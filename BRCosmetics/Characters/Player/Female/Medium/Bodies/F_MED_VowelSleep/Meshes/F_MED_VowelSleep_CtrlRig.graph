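# Control Rig for the F_MED_VowelSleep female body that runs FortDeformRig, then adjusts the belt and skirt bones.
# When IsClothEnabled is false, it counters thigh twist on helper bones that drive side, front and rear skirt bones.
# Center skirt bones blend and aim between their left/right neighbors.
# Spherical pose readers and thigh-up attributes weight corrective offsets, and a Kalman-smoothed IsKneeSliding flag blends in a knee-slide pose.
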ControlRigBlueprint /BRCosmetics/Characters/Player/Female/Medium/Bodies/F_MED_VowelSleep/Meshes/F_MED_VowelSleep_CtrlRig
graphs [
 CollapseNode_ContainedGraph {nodes [
   RigVMFunctionEntryNode()
   > RigVMFunction_Sequence = RigVMAggregateNode()
   ItemArray = RigVMDispatchNode(Value=((Type=Bone,Name="dyn_skirt_front_l_base"),(Type=Bone,Name="dyn_skirt_side_l_base"))) {template="DISPATCH_RigVMDispatch_Constant(io Value)"}
   For_Each = RigVMDispatchNode(Array=@ItemArray.Value, ExecuteContext=@RigVMFunction_Sequence.A) {template="DISPATCH_RigVMDispatch_ArrayIterator(in Array,out Element,out Index,out Count,out Ratio)"}
   > &RigUnit_ParentConstraint(
    AdvancedSettings=(InterpolationType=Average,RotationOrderForFilter=XZY)
    Parents=((Item=(Type=Bone,Name="cr_thigh_l_no_twist"),Weight=1))
    Weight=1
    Child=@For_Each.Element
   )
   ItemArray_1 = RigVMDispatchNode(Value=((Type=Bone,Name="dyn_skirt_front_r_base"),(Type=Bone,Name="dyn_skirt_side_r_base"))) {template="DISPATCH_RigVMDispatch_Constant(io Value)"}
   For_Each_1 = RigVMDispatchNode(Array=@ItemArray_1.Value, ExecuteContext=@RigVMFunction_Sequence.B) {template="DISPATCH_RigVMDispatch_ArrayIterator(in Array,out Element,out Index,out Count,out Ratio)"}
   > &RigUnit_ParentConstraint(
    AdvancedSettings=(InterpolationType=Average,RotationOrderForFilter=XZY)
    Parents=((Item=(Type=Bone,Name="cr_thigh_r_no_twist"),Weight=1))
    Weight=1
    Child=@For_Each_1.Element
   )
   RigVMFunctionReturnNode(ExecuteContext=@RigVMFunction_Sequence.H)
   Get_Animation_Attribute_3 = RigVMDispatchNode(BoneName=None, Default=0, Name=deformRig_r_thigh_up_alpha) {template="DISPATCH_RigDispatch_GetAnimAttribute(in Name,in BoneName,in Default,out Value,out Found)"}
   n1 = &RigVMFunction_MathDoubleRemap(SourceMaximum=1, SourceMinimum=0, TargetMaximum=0.8, TargetMinimum=0.5, Value=@Get_Animation_Attribute_3.Value)
   Get_Animation_Attribute_2 = RigVMDispatchNode(BoneName=None, Default=0, Name=deformRig_l_thigh_up_alpha) {template="DISPATCH_RigDispatch_GetAnimAttribute(in Name,in BoneName,in Default,out Value,out Found)"}
   Remap = &RigVMFunction_MathDoubleRemap(SourceMaximum=1, SourceMinimum=0, TargetMaximum=0.8, TargetMinimum=0.5, Value=@Get_Animation_Attribute_2.Value)
   Subtract = &RigVMFunction_MathDoubleSub(A=1, B=@Remap.Result)
   Add = &RigVMFunction_MathDoubleAdd(B=@n1.Result, A=@Subtract.Result)
   n2 = /Script/ControlRig.RigUnit_ParentConstraintMath(
    AdvancedSettings=(InterpolationType=Average)
    Parents=((Item=(Type=Bone,Name="dyn_skirt_front_l_base"),Weight=0.8),(Item=(Type=Bone,Name="dyn_skirt_front_r_base"),Weight=0.2))
    Parents.1.Weight=@Add.Result
    Parents.0.Weight=&RigVMFunction_MathDoubleAdd(A=@Remap.Result, B=&RigVMFunction_MathDoubleSub(A=1, B=@n1.Result).Result).Result
    Input=&RigUnit_GetTransform(Item=(Type=Bone,Name="dyn_skirt_front_base"), bInitial=true).Transform
   ) {is=struct RigVMUnitNode}
   n3 = &RigUnit_GetTransform(Item=(Type=Bone,Name="dyn_skirt_front_02"), bInitial=true)
   RigVMFunction_MathTransformTransformVector = /Script/RigVM.RigVMFunction_MathTransformTransformVector(Location=(X=20,Y=0,Z=0), Transform=@n3.Transform) {is=struct RigVMUnitNode}
   Get_Animation_Attribute_1 = RigVMDispatchNode(BoneName=None, Default=0, Name=deformRig_l_thigh_up_alpha) {template="DISPATCH_RigDispatch_GetAnimAttribute(in Name,in BoneName,in Default,out Value,out Found)"}
   Get_Animation_Attribute_6 = RigVMDispatchNode(BoneName=None, Default=0, Name=deformRig_r_thigh_up_alpha) {template="DISPATCH_RigDispatch_GetAnimAttribute(in Name,in BoneName,in Default,out Value,out Found)"}
   n4 = &RigUnit_ParentConstraintMath(
    AdvancedSettings=(InterpolationType=Average)
    Input=(Rotation=(X=0,Y=0,Z=0,W=1),Translation=(X=0,Y=0,Z=0),Scale3D=(X=1,Y=1,Z=1))
    Parents=((Item=(Type=Bone,Name="dyn_skirt_front_l_base"),Weight=0.5),(Item=(Type=Bone,Name="dyn_skirt_front_r_base"),Weight=0.5))
    Input.Rotation=@n3.Transform.Rotation
    Input.Translation=@RigVMFunction_MathTransformTransformVector.Result
    Parents.1.Weight=&RigVMFunction_MathDoubleSub(
     A=1
     B=&RigVMFunction_MathDoubleRemap(SourceMaximum=1, SourceMinimum=0, TargetMaximum=0.8, TargetMinimum=0.5, Value=@Get_Animation_Attribute_1.Value).Result
    ).Result
    Parents.0.Weight=&RigVMFunction_MathDoubleSub(
     A=1
     B=&RigVMFunction_MathDoubleRemap(SourceMaximum=1, SourceMinimum=0, TargetMaximum=0.8, TargetMinimum=0.5, Value=@Get_Animation_Attribute_6.Value).Result
    ).Result
   )
   n5 = /Script/ControlRig.RigUnit_AimBoneMath(
    DebugSettings=(bEnabled=False,Scale=10,WorldOffset=(Rotation=(X=0,Y=0,Z=0,W=1),Translation=(X=0,Y=0,Z=0),Scale3D=(X=1,Y=1,Z=1)))
    Primary=(Weight=1,Axis=(X=1,Y=0,Z=0),Target=(X=1,Y=0,Z=0),Kind=Location,Space=(Type=Bone,Name="None"))
    Secondary=(Weight=1,Axis=(X=0,Y=0,Z=-1),Target=(X=1,Y=0,Z=0),Kind=Location,Space=(Type=Bone,Name="None"))
    Weight=1
    InputTransform=@n2.Output
    Primary.Target=@n4.Output.Translation
    Secondary.Target=&RigUnit_GetTransform(Item=(Type=Bone,Name="dyn_skirt_front_l_base"), bInitial=False).Transform.Translation
   ) {is=struct RigVMUnitNode}
   RigUnit_SetTransform = &RigUnit_SetTransform(
    Item=(Type=Bone,Name="dyn_skirt_front_base")
    Value=(Rotation=(X=0.6519,Y=0.3124,Z=-0.634,W=-0.2747),Translation=(X=0.8505,Y=3.5922,Z=101.7967),Scale3D=(X=1,Y=1,Z=1))
    Weight=1
    bInitial=False
    Value.Rotation=@n5.Result.Rotation
    Value.Translation=@n2.Output.Translation
    ExecutePin=@RigVMFunction_Sequence.C
   )
   n6 = &RigUnit_SphericalPoseReader(
    ActiveRegionScaleFactors=(PositiveWidth=3,NegativeWidth=2,PositiveHeight=3,NegativeHeight=3)
    ActiveRegionSize=0.1
    Debug=(bDrawDebug=true,bDraw2D=False,bDrawLocalAxes=False,DebugScale=25,DebugSegments=20,DebugThickness=0.25)
    DriverAxis=(X=-1,Y=0,Z=0)
    DriverItem=(Type=Bone,Name="thigh_l")
    FalloffRegionScaleFactors=(PositiveWidth=1,NegativeWidth=0.95,PositiveHeight=1,NegativeHeight=1)
    FalloffSize=0.5
    FlipHeightScaling=False
    FlipWidthScaling=False
    OptionalParentItem=(Type=Bone,Name="pelvis")
    RotationOffset=(X=-90,Y=0,Z=0)
    ExecutePin=@RigVMFunction_Sequence.D
   )
   > &RigUnit_ParentConstraint(
    AdvancedSettings=(InterpolationType=Average,RotationOrderForFilter=XZY)
    Child=(Type=Bone,Name="dyn_skirt_rear_l_base")
    Parents=((Item=(Type=Bone,Name="cr_thigh_l_no_twist"),Weight=1))
    Weight=&RigVMFunction_MathDoubleRemap(SourceMaximum=0.3, SourceMinimum=0, TargetMaximum=1, TargetMinimum=0.25, Value=@n6.OutputParam).Result
   )
   > n7 = &RigUnit_SphericalPoseReader(
    ActiveRegionScaleFactors=(PositiveWidth=3,NegativeWidth=2,PositiveHeight=3,NegativeHeight=3)
    ActiveRegionSize=0.1
    Debug=(bDrawDebug=false,bDraw2D=False,bDrawLocalAxes=False,DebugScale=25,DebugSegments=20,DebugThickness=0.25)
    DriverAxis=(X=1,Y=0,Z=0)
    DriverItem=(Type=Bone,Name="thigh_r")
    FalloffRegionScaleFactors=(PositiveWidth=1,NegativeWidth=0.95,PositiveHeight=1,NegativeHeight=1)
    FalloffSize=0.5
    FlipHeightScaling=true
    FlipWidthScaling=true
    OptionalParentItem=(Type=Bone,Name="pelvis")
    RotationOffset=(X=90,Y=0,Z=0)
   )
   > &RigUnit_ParentConstraint(
    AdvancedSettings=(InterpolationType=Average,RotationOrderForFilter=XZY)
    Child=(Type=Bone,Name="dyn_skirt_rear_r_base")
    Parents=((Item=(Type=Bone,Name="cr_thigh_r_no_twist"),Weight=1))
    Weight=&RigVMFunction_MathDoubleRemap(SourceMaximum=0.3, SourceMinimum=0, TargetMaximum=1, TargetMinimum=0.25, Value=@n7.OutputParam).Result
   )
   n8 = &RigUnit_GetTransform(Item=(Type=Bone,Name="dyn_skirt_rear_03"), bInitial=true)
   n9 = &RigUnit_ParentConstraintMath(
    AdvancedSettings=(InterpolationType=Average)
    Input=(Rotation=(X=0,Y=0,Z=0,W=1),Translation=(X=0,Y=0,Z=0),Scale3D=(X=1,Y=1,Z=1))
    Parents=((Item=(Type=Bone,Name="dyn_skirt_rear_l_base"),Weight=0.5),(Item=(Type=Bone,Name="dyn_skirt_rear_r_base"),Weight=0.5))
    Input.Rotation=@n8.Transform.Rotation
    Input.Translation=&RigVMFunction_MathTransformTransformVector(Location=(X=20,Y=0,Z=0), Transform=@n8.Transform).Result
   )
   n10 = &RigUnit_ParentConstraintMath(
    AdvancedSettings=(InterpolationType=Average)
    Parents=((Item=(Type=Bone,Name="dyn_skirt_rear_l_base"),Weight=0.5),(Item=(Type=Bone,Name="dyn_skirt_rear_r_base"),Weight=0.5))
    Input=&RigUnit_GetTransform(Item=(Type=Bone,Name="dyn_skirt_rear_base"), bInitial=true).Transform
   )
   n11 = &RigUnit_AimBoneMath(
    DebugSettings=(bEnabled=False,Scale=10,WorldOffset=(Rotation=(X=0,Y=0,Z=0,W=1),Translation=(X=0,Y=0,Z=0),Scale3D=(X=1,Y=1,Z=1)))
    Primary=(Weight=1,Axis=(X=1,Y=0,Z=0),Target=(X=1,Y=0,Z=0),Kind=Location,Space=(Type=Bone,Name="None"))
    Secondary=(Weight=1,Axis=(X=0,Y=0,Z=1),Target=(X=1,Y=0,Z=0),Kind=Location,Space=(Type=Bone,Name="None"))
    Weight=1
    Primary.Target=@n9.Output.Translation
    InputTransform=@n10.Output
    Secondary.Target=&RigUnit_GetTransform(Item=(Type=Bone,Name="dyn_skirt_rear_l_base"), bInitial=false).Transform.Translation
   )
   > &RigUnit_SetTransform(
    Item=(Type=Bone,Name="dyn_skirt_rear_base")
    Value=(Rotation=(X=-0.1035,Y=0.7055,Z=-0.0476,W=0.6995),Translation=(X=-0.0002,Y=-11.9454,Z=99.7946),Scale3D=(X=1,Y=1,Z=1))
    Weight=1
    bInitial=False
    Value.Rotation=@n11.Result.Rotation
    Value.Translation=@n10.Output.Translation
   )
   > n12 = &RigUnit_SphericalPoseReader(
    ActiveRegionScaleFactors=(PositiveWidth=3,NegativeWidth=1,PositiveHeight=1,NegativeHeight=1)
    ActiveRegionSize=0.1
    Debug=(bDrawDebug=false,bDraw2D=False,bDrawLocalAxes=False,DebugScale=25,DebugSegments=20,DebugThickness=0.25)
    DriverAxis=(X=-1,Y=0,Z=0)
    DriverItem=(Type=Bone,Name="thigh_l")
    FalloffRegionScaleFactors=(PositiveWidth=1,NegativeWidth=1,PositiveHeight=1,NegativeHeight=1)
    FalloffSize=0.5
    FlipHeightScaling=False
    FlipWidthScaling=False
    OptionalParentItem=(Type=Bone,Name="None")
    RotationOffset=(X=180,Y=0,Z=0)
   )
   > n13 = &RigUnit_SphericalPoseReader(
    ActiveRegionScaleFactors=(PositiveWidth=3,NegativeWidth=1,PositiveHeight=1,NegativeHeight=1)
    ActiveRegionSize=0.1
    Debug=(bDrawDebug=false,bDraw2D=False,bDrawLocalAxes=False,DebugScale=25,DebugSegments=20,DebugThickness=0.25)
    DriverAxis=(X=1,Y=0,Z=0)
    DriverItem=(Type=Bone,Name="thigh_r")
    FalloffRegionScaleFactors=(PositiveWidth=1,NegativeWidth=1,PositiveHeight=1,NegativeHeight=1)
    FalloffSize=0.5
    FlipHeightScaling=true
    FlipWidthScaling=true
    OptionalParentItem=(Type=Bone,Name="None")
    RotationOffset=(X=0,Y=0,Z=0)
   )
   > Modify Transforms with Mirror = RigVMFunctionReferenceNode(
    Base Modify Data=((Item=(Type=Bone,Name="dyn_skirt_rear_l_base"),Transform=(Rotation=(X=0,Y=0,Z=-0.0436,W=0.999),Translation=(X=0,Y=-1,Z=0),Scale3D=(X=1,Y=1,Z=1))))
    Mirror Modify Data=((Item=(Type=Bone,Name="dyn_skirt_rear_r_base"),Transform=(Rotation=(X=0,Y=0,Z=0,W=1),Translation=(X=0,Y=0,Z=0),Scale3D=(X=1,Y=1,Z=1))))
    Weight=&RigVMFunction_MathDoubleRemap(SourceMaximum=0.3, SourceMinimum=0, TargetMaximum=1, TargetMinimum=0, Value=@n12.OutputParam).Result
    "Mirror Weight "=&RigVMFunction_MathDoubleRemap(SourceMaximum=0.3, SourceMinimum=0, TargetMaximum=1, TargetMinimum=0, Value=@n13.OutputParam).Result
   )
   > n14 = &RigUnit_SphericalPoseReader(
    ActiveRegionScaleFactors=(PositiveWidth=3,NegativeWidth=1,PositiveHeight=3,NegativeHeight=3)
    ActiveRegionSize=0.1
    Debug=(bDrawDebug=false,bDraw2D=False,bDrawLocalAxes=False,DebugScale=25,DebugSegments=20,DebugThickness=0.25)
    DriverAxis=(X=-1,Y=0,Z=0)
    DriverItem=(Type=Bone,Name="thigh_l")
    FalloffRegionScaleFactors=(PositiveWidth=1,NegativeWidth=0.95,PositiveHeight=1,NegativeHeight=1)
    FalloffSize=0.5
    FlipHeightScaling=False
    FlipWidthScaling=False
    OptionalParentItem=(Type=Bone,Name="None")
    RotationOffset=(X=-90,Y=0,Z=0)
   )
   > n15 = &RigUnit_SphericalPoseReader(
    ActiveRegionScaleFactors=(PositiveWidth=3,NegativeWidth=1,PositiveHeight=3,NegativeHeight=3)
    ActiveRegionSize=0.1
    Debug=(bDrawDebug=false,bDraw2D=False,bDrawLocalAxes=False,DebugScale=25,DebugSegments=20,DebugThickness=0.25)
    DriverAxis=(X=1,Y=0,Z=0)
    DriverItem=(Type=Bone,Name="thigh_r")
    FalloffRegionScaleFactors=(PositiveWidth=1,NegativeWidth=0.95,PositiveHeight=1,NegativeHeight=1)
    FalloffSize=0.5
    FlipHeightScaling=true
    FlipWidthScaling=true
    OptionalParentItem=(Type=Bone,Name="None")
    RotationOffset=(X=90,Y=0,Z=0)
   )
   > Modify Transforms with Mirror_1 = RigVMFunctionReferenceNode(
    Base Modify Data=((Item=(Type=Bone,Name="dyn_skirt_rear_l_base"),Transform=(Rotation=(X=0.0872,Y=0,Z=0,W=0.9962),Translation=(X=2,Y=1,Z=-2),Scale3D=(X=1,Y=1,Z=1))))
    Mirror Modify Data=((Item=(Type=Bone,Name="dyn_skirt_rear_r_base"),Transform=(Rotation=(X=0,Y=0,Z=0,W=1),Translation=(X=0,Y=0,Z=0),Scale3D=(X=1,Y=1,Z=1))))
    Weight=@n14.OutputParam
    "Mirror Weight "=@n15.OutputParam
   )
   Multiply = &RigVMFunction_MathDoubleMul(B=0.5, A=&RigVMFunction_MathDoubleAdd(A=@n14.OutputParam, B=@n15.OutputParam).Result)
   > &RigUnit_ModifyTransforms(
    ItemToModify=((Item=(Type=Bone,Name="dyn_skirt_rear_base"),Transform=(Rotation=(X=0.1113,Y=0.0284,Z=-0.3733,W=0.9206),Translation=(X=4.317,Y=-1.5324,Z=-0.2195),Scale3D=(X=1,Y=1,Z=1))))
    WeightMinimum=0
    Weight=@Multiply.Result
   )
   Get_Animation_Attribute_4 = RigVMDispatchNode(BoneName=None, Default=0, Name=deformRig_l_thigh_up_alpha) {template="DISPATCH_RigDispatch_GetAnimAttribute(in Name,in BoneName,in Default,out Value,out Found)"}
   &RigUnit_ModifyTransforms(
    ItemToModify="((Item=(Type=Bone,Name=\"dyn_skirt_rear_l_base\"),Transform=(Rotation=(X=0,Y=0,Z=0,W=1),Translation=(X=0,Y=-2,Z=0),Scale3D=(X=1,Y=1,Z=1))),(Item=(Type=Bone,Name=\"dyn_skirt_rear_base\"),Transform=(Rotation=(X=0,Y=0,Z=0,W=1),Translation=(X=0,Y=-1,Z=0),Scale3D=( ...(truncated)"
    WeightMinimum=0
    Weight=@Get_Animation_Attribute_4.Value
    ExecutePin=@RigVMFunction_Sequence.E
   )
   Get_Animation_Attribute_5 = RigVMDispatchNode(BoneName=None, Default=0, Name=deformRig_r_thigh_up_alpha) {template="DISPATCH_RigDispatch_GetAnimAttribute(in Name,in BoneName,in Default,out Value,out Found)"}
   > &RigUnit_ModifyTransforms(
    ItemToModify="((Item=(Type=Bone,Name=\"dyn_skirt_rear_r_base\"),Transform=(Rotation=(X=0,Y=0,Z=0,W=1),Translation=(X=0,Y=2,Z=0),Scale3D=(X=1,Y=1,Z=1))),(Item=(Type=Bone,Name=\"dyn_skirt_rear_base\"),Transform=(Rotation=(X=0,Y=0,Z=0,W=1),Translation=(X=0,Y=-1,Z=0),Scale3D=(X ...(truncated)"
    WeightMinimum=0
    Weight=@Get_Animation_Attribute_5.Value
   )
   > &RigUnit_ModifyTransforms(
    ItemToModify=((Item=(Type=Bone,Name="dyn_skirt_front_base"),Transform=(Rotation=(X=0,Y=0,Z=0,W=1),Translation=(X=1.5,Y=0,Z=0),Scale3D=(X=1,Y=1,Z=1))))
    WeightMinimum=0
    Weight=&RigVMFunction_MathDoubleMul(B=0.5, A=&RigVMFunction_MathDoubleAdd(A=@Get_Animation_Attribute_4.Value, B=@Get_Animation_Attribute_5.Value).Result).Result
   )
   KalmanFilter = &RigVMFunction_KalmanFloat(
    BufferSize=4
    Value=RigVMDispatchNode(False=0, True=1, Condition=RigVMVariableNode() {variable=IsKneeSliding}.Value) {template="DISPATCH_RigVMDispatch_If(in Condition,in True,in False,out Result)"}.Result
   ) {template="KalmanFilter::Execute(in Value,in BufferSize,out Result)"}
   &RigUnit_ModifyTransforms(
    ItemToModify="((Item=(Type=Bone,Name=\"dyn_skirt_rear_base\"),Transform=(Rotation=(X=0,Y=0,Z=-0.7373,W=0.6756),Translation=(X=-2,Y=0,Z=0),Scale3D=(X=1,Y=1,Z=1))),(Item=(Type=Bone,Name=\"dyn_skirt_rear_l_base\"),Transform=(Rotation=(X=0,Y=0,Z=-0.7373,W=0.6756),Translation=(X ...(truncated)"
    WeightMinimum=0
    Weight=@KalmanFilter.Result
    ExecutePin=@RigVMFunction_Sequence.F
   )
   &RigUnit_SetRotation(
    ExecutePin=()
    Item=(Type=Bone,Name="dyn_skirt_rear_base")
    Space=GlobalSpace
    Value=(X=0.0479,Y=0.7326,Z=0.0327,W=0.6782)
    Weight=1
    bInitial=False
   )
  ]}
 RigVMFunctionLibrary {links [], nodes []}
 RigVMFunction_Sequence_1_ContainedGraph {nodes [
   n1 = RigVMFunctionEntryNode()
   RigVMFunction_Sequence = &RigVMFunction_Sequence(ExecuteContext=@n1.ExecuteContext)
   n2 = &RigVMFunction_Sequence(ExecuteContext=@RigVMFunction_Sequence.B)
   n3 = &RigVMFunction_Sequence(ExecuteContext=@n2.B)
   n4 = &RigVMFunction_Sequence(ExecuteContext=@n3.B)
   n5 = &RigVMFunction_Sequence(ExecuteContext=@n4.B)
   n6 = &RigVMFunction_Sequence(ExecuteContext=@n5.B)
   n7 = &RigVMFunction_Sequence(ExecuteContext=@n6.B)
   RigVMFunctionReturnNode(A=@RigVMFunction_Sequence.A, B=@n2.A, C=@n3.A, D=@n4.A, E=@n5.A, F=@n6.A, G=@n7.A, H=@n7.B)
  ]}
 RigVMModel {nodes [
   RigUnit_BeginExecution = &RigUnit_BeginExecution()
   /Script/FortniteGame.RigUnit_FortDeformRig(
    Bool=(bUpperArm=true,bElbow=true,bWrist=true,bGroin=true,bGlute=True,bKnee=true,bPec=true,bBicep=False,bLat=False)
    Curve="(bLUpperArmUpAlpha=False,bRUpperArmUpAlpha=False,bLUpperArmFrAlpha=False,bRUpperArmFrAlpha=False,bLUpperArmBkAlpha=False,bRUpperArmBkAlpha=False,bLWristUpAlpha=False,bRWristUpAlpha=False,bLWristDnAlpha=False,bRWristDnAlpha=False,bLKneeAlpha=False,bRKneeAlp ...(truncated)"
    Debug=(bUpperArmDebug=False,bWristDebug=False,bLatDebug=False,bGroinGluteDebug=False)
    Value="(UpperArmUp=(Rotation=(X=0.0,Y=-0.0,Z=0.0,W=1.0),Translation=(X=0.0,Y=0.0,Z=8.0),Scale3D=(X=0.70,Y=1.0,Z=1.0)),UpperArmFwdFr=(Rotation=(X=0.0,Y=-0.0,Z=0.0,W=1.0),Translation=(X=0.0,Y=10.0,Z=0.0),Scale3D=(X=0.50,Y=1.0,Z=1.0)),UpperArmFwdBk=(Rotation=(X=0.0, ...(truncated)"
    ExecutePin=@RigUnit_BeginExecution.ExecutePin
   ) {is=struct RigVMUnitNode}
   > n1 = &RigUnit_ParentConstraint(
    AdvancedSettings=(InterpolationType=Average,RotationOrderForFilter=XZY)
    Child=(Type=Bone,Name="def_belt")
    Parents=((Item=(Type=Bone,Name="spine_03"),Weight=1),(Item=(Type=Bone,Name="spine_02"),Weight=1),(Item=(Type=Bone,Name="pelvis"),Weight=1))
    Weight=1
   )
   RigVMFunction_ControlFlowBranch = &RigVMFunction_ControlFlowBranch(ExecuteContext=@n1.ExecutePin, Condition=RigVMVariableNode() {variable=IsClothEnabled}.Value)
   Get_Animation_Attribute_2 = RigVMDispatchNode(BoneName=None, Default=0, Name=deformRig_l_thigh_up_alpha) {template="DISPATCH_RigDispatch_GetAnimAttribute(in Name,in BoneName,in Default,out Value,out Found)"}
   Get_Animation_Attribute_3 = RigVMDispatchNode(BoneName=None, Default=0, Name=deformRig_r_thigh_up_alpha) {template="DISPATCH_RigDispatch_GetAnimAttribute(in Name,in BoneName,in Default,out Value,out Found)"}
   ArrayAverage = /Script/RigVM.RigVMFunction_MathDoubleArrayAverage(Array=(0,0), Array.0=@Get_Animation_Attribute_2.Value, Array.1=@Get_Animation_Attribute_3.Value) {is=struct RigVMUnitNode, template="ArrayAverage::Execute(in Array,out Average)"}
   &RigUnit_ModifyTransforms(
    ItemToModify=((Item=(Type=Bone,Name="def_belt"),Transform=(Rotation=(X=0,Y=0,Z=0,W=1),Translation=(X=1,Y=0,Z=0),Scale3D=(X=1,Y=1,Z=1))))
    WeightMinimum=0
    Weight=@ArrayAverage.Average
    ExecutePin=@RigVMFunction_ControlFlowBranch.Completed
   )
   RigUnit_GetTransform = &RigUnit_GetTransform(Item=(Type=Bone,Name="thigh_l"), Space=LocalSpace, bInitial=False)
   n2 = &RigUnit_GetTransform(Item=(Type=Bone,Name="thigh_l"), Space=LocalSpace, bInitial=true)
   Inverse = /Script/RigVM.RigVMFunction_MathQuaternionInverse(Value=&RigVMFunction_MathQuaternionSwingTwist(
     TwistAxis=(X=1,Y=0,Z=0)
     Input=&RigVMFunction_MathTransformMakeRelative(Global=@RigUnit_GetTransform.Transform, Parent=@n2.Transform).Local.Rotation
    ).Twist) {is=struct RigVMUnitNode, template="Inverse::Execute(in Value,out Result)"}
   RigUnit_SetRotation = &RigUnit_SetRotation(
    Item=(Type=Bone,Name="cr_thigh_l_no_twist")
    Weight=1
    bInitial=False
    Value=@Inverse.Result
    ExecutePin=@RigVMFunction_ControlFlowBranch.False
   )
   n3 = &RigVMFunction_MathTransformMakeRelative(
    Global=&RigUnit_GetTransform(Item=(Type=Bone,Name="thigh_r"), Space=LocalSpace, bInitial=False).Transform
    Parent=&RigUnit_GetTransform(Item=(Type=Bone,Name="thigh_r"), Space=LocalSpace, bInitial=true).Transform
   )
   > &RigUnit_SetRotation(
    Item=(Type=Bone,Name="cr_thigh_r_no_twist")
    Weight=1
    bInitial=False
    Value=&RigVMFunction_MathQuaternionInverse(Value=&RigVMFunction_MathQuaternionSwingTwist(TwistAxis=(X=1,Y=0,Z=0), Input=@n3.Local.Rotation).Twist) {template="Inverse::Execute(in Value,out Result)"}.Result
   )
   > Skirt Deformation = RigVMCollapseNode()
  ]}
]
hierarchy [
 root {loc=<0 0 0>, rot=<0 0 0 1>}
 pelvis {parent=root, loc=<0 -0.5301 98.0571>, rot=<0.0127 0.707 -0.0127 -0.707>}
 spine_01 {parent=pelvis, loc=<6.1208 -0.9328 0>, rot=<0 0 0.034 -0.9994>}
 spine_02 {parent=spine_01, loc=<6.5997 -0.2095 0>, rot=<0 0 0.0238 -0.9997>}
 spine_03 {parent=spine_02, loc=<6.4914 -0.1237 0>, rot=<0 0 0.0149 -0.9999>}
 spine_04 {parent=spine_03, loc=<6.9598 0.0794 0>, rot=<0 0 -0.0132 -0.9999>}
 spine_05 {parent=spine_04, loc=<6.9987 0 0>, rot=<0 0 -0.0015 -1>}
 neck_01 {parent=spine_05, loc=<11.3138 0.1529 0>, rot=<0 0 -0.2003 -0.9797>}
 neck_02 {parent=neck_01, loc=<4.5627 0 0>, rot=<0 0 0.0504 -0.9987>}
 head {parent=neck_02, loc=<5.3914 0 0>, rot=<0 0 0.115 -0.9934>}
 clavicle_l {parent=spine_05, loc=<5.4831 -0.1053 -3.2946>, rot=<0.0465 -0.7289 -0.009 -0.6829>}
 upperarm_l {parent=clavicle_l, loc=<9.8658 -1.2136 -0.4105>, rot=<0.0164 -0.3245 -0.0269 -0.9454>}
 lowerarm_l {parent=upperarm_l, loc=<26.465 -0.8345 0.259>, rot=<0.0036 0.0161 -0.2298 -0.9731>}
 hand_l {parent=lowerarm_l, loc=<22.7355 -0.3203 -0.6105>, rot=<-0.6784 -0.0223 0.0735 -0.7307>}
 thumb_01_l {parent=hand_l, loc=<1.3118 -0.3191 -2.459>, rot=<0.6546 -0.4042 0.0664 -0.6354>}
 thumb_02_l {parent=thumb_01_l, loc=<5.9336 -0.0448 0.0401>, rot=<-0.0037 0.0213 0.298 -0.9543>}
 thumb_03_l {parent=thumb_02_l, loc=<2.2223 0.3682 -0.0907>, rot=<0.0118 -0.013 0.0066 -0.9998>}
 index_metacarpal_l {parent=hand_l, loc=<1.6535 0.0183 -2.96>, rot=<0.0158 -0.0124 -0.0173 -0.9996>}
 index_01_l {parent=index_metacarpal_l, loc=<8.3008 0.0357 -0.0524>, rot=<-0.0044 0.0485 0.0924 -0.9945>}
 index_02_l {parent=index_01_l, loc=<3.3587 -0.3004 0.0452>, rot=<-0.002 0.0027 0.3225 -0.9466>}
 index_03_l {parent=index_02_l, loc=<2.5578 -0.123 -0.0252>, rot=<-0.002 0.0067 0.2558 -0.9667>}
 middle_metacarpal_l {parent=hand_l, loc=<2.1856 -0.1725 -0.5984>, rot=<0.006 0.0168 -0.0329 -0.9993>}
 middle_01_l {parent=middle_metacarpal_l, loc=<7.3275 -0.2447 0.0129>, rot=<-0.0208 0.0042 0.0858 -0.9961>}
 middle_02_l {parent=middle_01_l, loc=<4.0123 0.1401 0.1175>, rot=<-0.0218 0.0258 0.3512 -0.9357>}
 middle_03_l {parent=middle_02_l, loc=<2.6077 0.0534 0.1373>, rot=<0.0082 0.0188 0.1708 -0.9851>}
 ring_metacarpal_l {parent=hand_l, loc=<2.2074 -0.3625 1.3638>, rot=<-0.0284 0.0722 -0.0458 -0.9959>}
 ring_01_l {parent=ring_metacarpal_l, loc=<7.0295 -0.6197 -0.1446>, rot=<-0.0489 -0.0252 0.1892 -0.9804>}
 ring_02_l {parent=ring_01_l, loc=<3.6174 0.4647 -0.1027>, rot=<0 0 0.2806 -0.9598>}
 ring_03_l {parent=ring_02_l, loc=<2.4152 0.114 0.0199>, rot=<0 0 0.1189 -0.9929>}
 pinky_metacarpal_l {parent=hand_l, loc=<1.7616 -0.9753 3.3814>, rot=<-0.0397 0.0814 -0.069 -0.9935>}
 pinky_01_l {parent=pinky_metacarpal_l, loc=<7.1709 -0.4831 -0.1094>, rot=<-0.0654 -0.0035 0.2242 -0.9723>}
 pinky_02_l {parent=pinky_01_l, loc=<2.8274 0.0893 -0.1043>, rot=<0.0037 0.0011 0.2854 -0.9584>}
 pinky_03_l {parent=pinky_02_l, loc=<2.2046 0.0307 -0.0078>, rot=<-0.0002 0.0098 0.1445 -0.9895>}
 deform_wrist_dn_base_l {parent=hand_l, loc=<0 -1.5326 0>, rot=<0 0 0 -1>}
 deform_wrist_dn_l {parent=deform_wrist_dn_base_l, loc=<0 0 0>, rot=<0 0 0 -1>}
 deform_wrist_up_base_l {parent=hand_l, loc=<0 2.3009 0>, rot=<0 0 0 -1>}
 deform_wrist_up_l {parent=deform_wrist_up_base_l, loc=<0 0 0>, rot=<0 0 0 -1>}
 lowerarm_twist_02_l {parent=lowerarm_l, loc=<10.6707 0 0>, rot=<0 0 0 1>}
 lowerarm_twist_01_l {parent=lowerarm_l, loc=<18.316 0 0>, rot=<0 0 0 1>}
 deform_elbow_in_base_l {parent=lowerarm_l, loc=<0.1834 3.6062 -0.2572>, rot=<0 0 0 -1>}
 deform_elbow_in_l {parent=deform_elbow_in_base_l, loc=<0 0 0>, rot=<0 0 0 -1>}
 deform_elbow_out_base_l {parent=lowerarm_l, loc=<0 -4 0>, rot=<0 0 0 1>}
 deform_elbow_out_l {parent=deform_elbow_out_base_l, loc=<0 0 0>, rot=<0 0 0 1>}
 upperarm_twist_01_l {parent=upperarm_l, loc=<7.2135 -0.1134 0>, rot=<0 0 0 1>}
 deform_upperarm_bk_base_l {parent=upperarm_twist_01_l, loc=<-6.4053 -4.3746 0.314>, rot=<0 0 0 -1>}
 deform_upperarm_bk_l {parent=deform_upperarm_bk_base_l, loc=<0 0 0>, rot=<0 0 0 -1>}
 deform_upperarm_fr_base_l {parent=upperarm_twist_01_l, loc=<-5.6312 5.092 0.3354>, rot=<0 0 0 -1>}
 deform_upperarm_fr_l {parent=deform_upperarm_fr_base_l, loc=<0 0 0>, rot=<0 0 0 -1>}
 deform_upperarm_up_base_l {parent=upperarm_twist_01_l, loc=<-8.891 0.4825 3.9517>, rot=<0 0 0 -1>}
 deform_upperarm_up_l {parent=deform_upperarm_up_base_l, loc=<0 0 0>, rot=<0 0 0 -1>}
 upperarm_twist_02_l {parent=upperarm_l, loc=<19.7998 -0.1134 0>, rot=<0 0 0 1>}
 clavicle_r {parent=spine_05, loc=<5.4831 -0.1053 3.2946>, rot=<0.7289 0.0465 -0.6829 0.009>}
 upperarm_r {parent=clavicle_r, loc=<-9.8658 1.2136 0.4105>, rot=<0.0164 -0.3245 -0.0269 -0.9454>}
 lowerarm_r {parent=upperarm_r, loc=<-26.465 0.8345 -0.259>, rot=<0.0036 0.0161 -0.2298 -0.9731>}
 hand_r {parent=lowerarm_r, loc=<-22.7355 0.3203 0.6106>, rot=<-0.6784 -0.0223 0.0735 -0.7307>}
 thumb_01_r {parent=hand_r, loc=<-1.3118 0.3191 2.459>, rot=<0.6546 -0.4042 0.0664 -0.6354>}
 thumb_02_r {parent=thumb_01_r, loc=<-5.9336 0.0448 -0.0401>, rot=<-0.0037 0.0213 0.298 -0.9543>}
 thumb_03_r {parent=thumb_02_r, loc=<-2.2223 -0.3682 0.0907>, rot=<0.0118 -0.013 0.0066 -0.9998>}
 index_metacarpal_r {parent=hand_r, loc=<-1.6535 -0.0183 2.96>, rot=<0.0158 -0.0124 -0.0173 -0.9996>}
 index_01_r {parent=index_metacarpal_r, loc=<-8.3008 -0.0357 0.0524>, rot=<-0.0044 0.0485 0.0924 -0.9945>}
 index_02_r {parent=index_01_r, loc=<-3.3587 0.3004 -0.0452>, rot=<-0.002 0.0027 0.3225 -0.9466>}
 index_03_r {parent=index_02_r, loc=<-2.5578 0.123 0.0252>, rot=<-0.002 0.0067 0.2558 -0.9667>}
 middle_metacarpal_r {parent=hand_r, loc=<-2.1856 0.1725 0.5984>, rot=<0.006 0.0168 -0.0329 -0.9993>}
 middle_01_r {parent=middle_metacarpal_r, loc=<-7.3275 0.2447 -0.0129>, rot=<-0.0208 0.0042 0.0858 -0.9961>}
 middle_02_r {parent=middle_01_r, loc=<-4.0123 -0.1401 -0.1175>, rot=<-0.0218 0.0258 0.3512 -0.9357>}
 middle_03_r {parent=middle_02_r, loc=<-2.6077 -0.0534 -0.1373>, rot=<0.0082 0.0188 0.1708 -0.9851>}
 ring_metacarpal_r {parent=hand_r, loc=<-2.2074 0.3625 -1.3638>, rot=<-0.0284 0.0722 -0.0458 -0.9959>}
 ring_01_r {parent=ring_metacarpal_r, loc=<-7.0295 0.6197 0.1446>, rot=<-0.0489 -0.0252 0.1892 -0.9804>}
 ring_02_r {parent=ring_01_r, loc=<-3.6174 -0.4647 0.1027>, rot=<0 0 0.2806 -0.9598>}
 ring_03_r {parent=ring_02_r, loc=<-2.4152 -0.114 -0.0199>, rot=<0 0 0.1189 -0.9929>}
 pinky_metacarpal_r {parent=hand_r, loc=<-1.7616 0.9753 -3.3814>, rot=<-0.0397 0.0814 -0.069 -0.9935>}
 pinky_01_r {parent=pinky_metacarpal_r, loc=<-7.1709 0.4831 0.1094>, rot=<-0.0654 -0.0035 0.2242 -0.9723>}
 pinky_02_r {parent=pinky_01_r, loc=<-2.8274 -0.0893 0.1043>, rot=<0.0037 0.0011 0.2854 -0.9584>}
 pinky_03_r {parent=pinky_02_r, loc=<-2.2046 -0.0307 0.0078>, rot=<-0.0002 0.0098 0.1445 -0.9895>}
 deform_wrist_dn_base_r {parent=hand_r, loc=<-0.0001 1.5328 0>, rot=<0 0 0 -1>}
 deform_wrist_dn_r {parent=deform_wrist_dn_base_r, loc=<0 0 0>, rot=<0 0 0 -1>}
 deform_wrist_up_base_r {parent=hand_r, loc=<-0.0002 -2.3007 -0.0001>, rot=<0 0 0 -1>}
 deform_wrist_up_r {parent=deform_wrist_up_base_r, loc=<0 0 0>, rot=<0 0 0 -1>}
 lowerarm_twist_02_r {parent=lowerarm_r, loc=<-10.6707 0 0>, rot=<0 0 0 -1>}
 lowerarm_twist_01_r {parent=lowerarm_r, loc=<-18.316 0 0>, rot=<0 0 0 -1>}
 deform_elbow_in_base_r {parent=lowerarm_r, loc=<-0.1834 -3.6062 0.2572>, rot=<0 0 0 -1>}
 deform_elbow_in_r {parent=deform_elbow_in_base_r, loc=<0 0 0>, rot=<0 0 0 -1>}
 deform_elbow_out_base_r {parent=lowerarm_r, loc=<0 4 -0.0001>, rot=<0 0 0 -1>}
 deform_elbow_out_r {parent=deform_elbow_out_base_r, loc=<0 0 0>, rot=<0 0 0 1>}
 upperarm_twist_01_r {parent=upperarm_r, loc=<-7.2135 0.1134 0>, rot=<0 0 0 -1>}
 deform_upperarm_bk_base_r {parent=upperarm_twist_01_r, loc=<6.4053 4.3746 -0.314>, rot=<0 0 0 -1>}
 deform_upperarm_bk_r {parent=deform_upperarm_bk_base_r, loc=<0 0 0>, rot=<0 0 0 -1>}
 deform_upperarm_fr_base_r {parent=upperarm_twist_01_r, loc=<5.6312 -5.092 -0.3354>, rot=<0 0 0 -1>}
 deform_upperarm_fr_r {parent=deform_upperarm_fr_base_r, loc=<0 0 0>, rot=<0 0 0 -1>}
 deform_upperarm_up_base_r {parent=upperarm_twist_01_r, loc=<8.891 -0.4825 -3.9517>, rot=<0 0 0 -1>}
 deform_upperarm_up_r {parent=deform_upperarm_up_base_r, loc=<0 0 0>, rot=<0 0 0 -1>}
 upperarm_twist_02_r {parent=upperarm_r, loc=<-19.7998 0.1134 0>, rot=<0 0 0 -1>}
 attach_backpack {parent=spine_05, loc=<-0.123 -13.473 0>, rot=<-0.0093 -0.707 0.0093 -0.707>}
 attach_cape {parent=spine_05, loc=<9.017 -11.714 0>, rot=<0.0616 -0.7044 -0.0616 -0.7044>}
 Backpack_BR {parent=spine_05, type=Null, loc=<0 0 0>, rot=<0 0 0 1>}
 cape {parent=spine_05, type=Null, loc=<0 0 0>, rot=<0 0 0 1>}
 deform_pec_aim_base_l {parent=spine_04, loc=<3.2456 9.3309 -1.2237>, rot=<0.0603 -0.4033 0.3318 -0.8507>}
 deform_pec_aim_l {parent=deform_pec_aim_base_l, loc=<0 0 0>, rot=<0 0 0 1>}
 deform_pec_base_l {parent=deform_pec_aim_l, loc=<9.3665 0 0>, rot=<0 0 0 1>}
 deform_pec_l {parent=deform_pec_base_l, loc=<0 0 0>, rot=<0 0 0 -1>}
 deform_pec_aim_base_r {parent=spine_04, loc=<3.2452 9.3309 1.2237>, rot=<-0.4033 -0.0603 0.8507 0.3318>}
 deform_pec_aim_r {parent=deform_pec_aim_base_r, loc=<0 0 0>, rot=<0 0 0 1>}
 deform_pec_base_r {parent=deform_pec_aim_r, loc=<-9.3667 -0.0003 -0.0004>, rot=<0 0 0 1>}
 deform_pec_r {parent=deform_pec_base_r, loc=<0 0 0>, rot=<0 0 0 -1>}
 def_belt {parent=spine_01, loc=<3.8216 1.141 0>, rot=<0 0 0 1>}
 def_belt_front {parent=def_belt, loc=<0 1 0>, rot=<0 0 0 1>}
 thigh_l {parent=pelvis, loc=<-1.0855 -0.2109 -8.7053>, rot=<0.0009 0.0355 0.0155 -0.9993>}
 calf_l {parent=thigh_l, loc=<-42.5284 -0.0019 -0.7297>, rot=<0.0006 -0.0161 -0.0391 -0.9991>}
 foot_l {parent=calf_l, loc=<-46.7401 -2.5014 -0.592>, rot=<0.0427 -0.0238 0.0353 -0.9982>}
 ball_l {parent=foot_l, loc=<-3.5658 10.8066 -0.1544>, rot=<0.0001 -0.0001 -0.7063 -0.7079>}
 calf_twist_02_l {parent=calf_l, loc=<-19.1546 -1.7943 -0.2638>, rot=<0.0237 0.0095 -0.0035 -0.9997>}
 calf_twist_01_l {parent=calf_l, loc=<-35.8483 -1.9197 -0.5794>, rot=<0.0237 0.0095 -0.0035 -0.9997>}
 deform_knee_in_base_l {parent=calf_l, loc=<0.8031 -5.0231 0.2456>, rot=<0 0 0 -1>}
 deform_knee_in_l {parent=deform_knee_in_base_l, loc=<0 0 0>, rot=<0 0 0 -1>}
 deform_knee_out_base_l {parent=calf_l, loc=<0 4 0>, rot=<0 0 0 -1>}
 deform_knee_out_l {parent=deform_knee_out_base_l, loc=<0 0 0>, rot=<0 0 0 -1>}
 thigh_twist_01_l {parent=thigh_l, loc=<-27.952 0 0>, rot=<0.0013 0 0 -1>}
 deform_groin_base_l {parent=thigh_l, loc=<0.4801 5.6776 -0.2486>, rot=<0 0 0 -1>}
 deform_groin_l {parent=deform_groin_base_l, loc=<0 0 0>, rot=<0 0 0 -1>}
 cr_thigh_l_no_twist {parent=thigh_l, loc=<0 0 0>, rot=<0 0 0 1>}
 thigh_r {parent=pelvis, loc=<-1.0855 -0.2109 8.7053>, rot=<0.0355 -0.0009 0.9993 0.0155>}
 calf_r {parent=thigh_r, loc=<42.5284 0.0019 0.7297>, rot=<0.0006 -0.0161 -0.0391 -0.9991>}
 foot_r {parent=calf_r, loc=<46.7401 2.5014 0.592>, rot=<0.0427 -0.0238 0.0353 -0.9982>}
 ball_r {parent=foot_r, loc=<3.5658 -10.8066 0.1544>, rot=<0.0001 -0.0001 -0.7063 -0.7079>}
 calf_twist_02_r {parent=calf_r, loc=<19.1546 1.7943 0.2638>, rot=<0.0237 0.0095 -0.0035 -0.9997>}
 calf_twist_01_r {parent=calf_r, loc=<35.8483 1.9197 0.5794>, rot=<0.0237 0.0095 -0.0035 -0.9997>}
 deform_knee_in_base_r {parent=calf_r, loc=<-0.8031 5.0231 -0.2457>, rot=<0 0 0 -1>}
 deform_knee_in_r {parent=deform_knee_in_base_r, loc=<0 0 0>, rot=<0 0 0 -1>}
 deform_knee_out_base_r {parent=calf_r, loc=<0 -4 0>, rot=<0 0 0 -1>}
 deform_knee_out_r {parent=deform_knee_out_base_r, loc=<0 0 0>, rot=<0 0 0 1>}
 thigh_twist_01_r {parent=thigh_r, loc=<27.952 0 0>, rot=<0.0013 0 0 -1>}
 deform_groin_base_r {parent=thigh_r, loc=<-0.4801 -5.6776 0.2486>, rot=<0 0 0 -1>}
 deform_groin_r {parent=deform_groin_base_r, loc=<0 0 0>, rot=<0 0 0 -1>}
 cr_thigh_r_no_twist {parent=thigh_r, loc=<0 0 0>, rot=<0 0 0 1>}
 deform_glute_base_l {parent=pelvis, loc=<-7.3466 -9.0213 -9.1245>, rot=<0.0172 0.031 -0.4477 -0.8935>}
 deform_glute_l {parent=deform_glute_base_l, loc=<0 0 0>, rot=<0 0 0 1>}
 deform_glute_base_r {parent=pelvis, loc=<-7.3464 -9.0213 9.1245>, rot=<0.031 -0.0172 0.8935 -0.4477>}
 deform_glute_r {parent=deform_glute_base_r, loc=<0 0 0>, rot=<0 0 0 1>}
 dyn_skirt_front_base {parent=pelvis, loc=<0.2233 9.1985 -1.3187>, rot=<0 0.0004 0.9977 0.068>}
 dyn_skirt_front_01 {parent=dyn_skirt_front_base, loc=<0.2 0 0>, rot=<0 0 0 1>}
 dyn_skirt_front_02 {parent=dyn_skirt_front_01, loc=<8.4026 0 0>, rot=<0 0 0 1>}
 dyn_skirt_front_03 {parent=dyn_skirt_front_02, loc=<6.4049 0 0>, rot=<0 0 0 1>}
 dyn_skirt_front_l_base {parent=pelvis, loc=<1.0958 7.1586 -10.6422>, rot=<-0.0511 0.1461 0.9813 0.1147>}
 dyn_skirt_front_l_01 {parent=dyn_skirt_front_l_base, loc=<0.2 0 0>, rot=<0 0 0 1>}
 dyn_skirt_front_l_02 {parent=dyn_skirt_front_l_01, loc=<9.3793 0 0>, rot=<0 0 0 1>}
 dyn_skirt_front_l_03 {parent=dyn_skirt_front_l_02, loc=<6.4558 0 0>, rot=<0 0 0 1>}
 dyn_skirt_side_l_base {parent=pelvis, loc=<2.0804 -1.3917 -13.8231>, rot=<0.1296 -0.7125 -0.6761 -0.1357>}
 dyn_skirt_side_l_01 {parent=dyn_skirt_side_l_base, loc=<0.2 0 0>, rot=<0 0 0 1>}
 dyn_skirt_side_l_02 {parent=dyn_skirt_side_l_01, loc=<11.3782 0 0>, rot=<0 0 0 1>}
 dyn_skirt_side_l_03 {parent=dyn_skirt_side_l_02, loc=<6.6771 0 0>, rot=<0 0 0 1>}
 dyn_skirt_rear_l_base {parent=pelvis, loc=<1.2435 -10.7679 -8.8627>, rot=<0.1208 -0.9722 -0.1928 -0.0552>}
 dyn_skirt_rear_l_01 {parent=dyn_skirt_rear_l_base, loc=<0.2 0 0>, rot=<0 0 0 1>}
 dyn_skirt_rear_l_02 {parent=dyn_skirt_rear_l_01, loc=<10.4568 0 0>, rot=<0 0 0 1>}
 dyn_skirt_rear_l_03 {parent=dyn_skirt_rear_l_02, loc=<6.3496 0 0>, rot=<0 0 0 1>}
 dyn_skirt_rear_base {parent=pelvis, loc=<1.3274 -11.4702 0.0002>, rot=<0.0891 -0.996 0.0004 0>}
 dyn_skirt_rear_01 {parent=dyn_skirt_rear_base, loc=<0.2 0 0>, rot=<0 0 0 1>}
 dyn_skirt_rear_02 {parent=dyn_skirt_rear_01, loc=<10.4379 0 0>, rot=<0 0 0 1>}
 dyn_skirt_rear_03 {parent=dyn_skirt_rear_02, loc=<6.6771 0 0>, rot=<0 0 0 1>}
 dyn_skirt_front_r_base {parent=pelvis, loc=<1.0958 7.1586 10.6422>, rot=<-0.1461 -0.0511 0.1147 -0.9813>}
 dyn_skirt_front_r_01 {parent=dyn_skirt_front_r_base, loc=<-0.2 0 0>, rot=<0 0 0 -1>}
 dyn_skirt_front_r_02 {parent=dyn_skirt_front_r_01, loc=<-9.3793 0 0>, rot=<0 0 0 -1>}
 dyn_skirt_front_r_03 {parent=dyn_skirt_front_r_02, loc=<-6.4557 0 0>, rot=<0 0 0 1>}
 dyn_skirt_side_r_base {parent=pelvis, loc=<2.0803 -1.3917 13.8231>, rot=<-0.7125 -0.1296 0.1357 -0.6761>}
 dyn_skirt_side_r_01 {parent=dyn_skirt_side_r_base, loc=<-0.2002 0 0>, rot=<0 0 0 1>}
 dyn_skirt_side_r_02 {parent=dyn_skirt_side_r_01, loc=<-11.378 0.0001 0>, rot=<0 0 0 1>}
 dyn_skirt_side_r_03 {parent=dyn_skirt_side_r_02, loc=<-6.6772 0 0>, rot=<0 0 0 1>}
 dyn_skirt_rear_r_base {parent=pelvis, loc=<1.2435 -10.7679 8.8627>, rot=<0.9722 0.1208 -0.0552 0.1928>}
 dyn_skirt_rear_r_01 {parent=dyn_skirt_rear_r_base, loc=<-0.2 0 0>, rot=<0 0 0 -1>}
 dyn_skirt_rear_r_02 {parent=dyn_skirt_rear_r_01, loc=<-10.4568 0 0>, rot=<0 0 0 -1>}
 dyn_skirt_rear_r_03 {parent=dyn_skirt_rear_r_02, loc=<-6.3496 0 -0.0001>, rot=<0 0 0 -1>}
]
variables [
 IsClothEnabled {type=bool, default=False}
 IsKneeSliding {type=bool, default=False}
]
