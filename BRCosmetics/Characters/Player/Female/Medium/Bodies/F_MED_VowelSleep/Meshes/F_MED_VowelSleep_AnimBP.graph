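# Post-process-style AnimBP for the F_MED_VowelSleep cosmetic body, driving its deformation and skirt motion.
# Copies the parent mesh pose into a Control Rig fed cloth and delayed knee-slide flags.
# With cloth enabled that pose is output; otherwise chained per-bone AnimDynamics simulate the skirt as fallback.
# Each update runs the cloth controller with wind gusts; crouch transitions briefly raise anim drive before restoring default cloth settings.
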AnimBlueprint /BRCosmetics/Characters/Player/Female/Medium/Bodies/F_MED_VowelSleep/Meshes/F_MED_VowelSleep_AnimBP
target_skeleton=&/F_MED_VowelSleep_Skeleton
graphs [
 AnimGraph {nodes [
   Output Pose(Result=Blend Poses by bool(
     BlendPose_0=Use cached pose 'PostDeformation'()
     BlendPose_1=Use cached pose 'PostAnimDynSkirt'()
     BlendTime_0=0
     BlendTime_1=0
     bActiveValue=Get bClothEnabled()
    )) {layerGroup=None}
   n1 = Control Rig!(
    Source=Copy Pose From Mesh() {bUseAttachedParent=true}
    Alpha
    IsKneeSliding:bool=Get DelayedKneeSlide() -> (DelayedKneeSlide:bool)
    IsClothEnabled:bool=Get bClothEnabled()
   ) -> (Pose) {controlRigAssetReference {blueprintRigClass=&/F_MED_VowelSleep_CtrlRig.F_MED_VowelSleep_CtrlRig_C}}
   Save cached pose 'PostDeformation'(Pose=@n1) {cacheName=PostDeformation}
   n2 = Anim Dynamics - Bone: dyn_skirt_front_01(ComponentPose=Local To Component(LocalPose=Use cached pose 'PostDeformation'()), Alpha=1, ExternalForce="0,0,0") {
    angularDampingOverride=0.999
    angularSpringConstant=20
    bAngularSpring=true
    bOverrideAngularDamping=true
    bUseGravityOverride=true
    gravityScale=0.35
    simulationSpace=RootRelative
    boundBone {boneName=dyn_skirt_front_01}
    componentAppliedLinearAccClamp=<0 0 0>
    gravityOverride {z=-1800}
    physicsBodyDefinitions [{
      boundBone {boneName=dyn_skirt_front_01}
      boxExtents=<15 15 30>
      constraintSetup {angularLimitsMax=<8 8 0>, angularLimitsMin=<-8 -8 -18>, angularTarget=<1 0 0>}
      localJointOffset=<4 0 0>
     }]
    simSpaceSettings {maxAngularAcceleration=4, maxAngularVelocity=400, simSpaceAngularAlpha=1}
   }
   n3 = Anim Dynamics - Bone: dyn_skirt_front_02(ComponentPose=@n2, Alpha=1, ExternalForce="0,0,0") {
    angularDampingOverride=0.999
    angularSpringConstant=20
    bAngularSpring=true
    bOverrideAngularDamping=true
    bUseGravityOverride=true
    gravityScale=0.35
    simulationSpace=RootRelative
    boundBone {boneName=dyn_skirt_front_02}
    componentAppliedLinearAccClamp=<0 0 0>
    gravityOverride {z=-1800}
    physicsBodyDefinitions [{
      boundBone {boneName=dyn_skirt_front_02}
      boxExtents=<15 15 30>
      constraintSetup {angularLimitsMax=<8 8 0>, angularLimitsMin=<-8 -8 -18>, angularTarget=<1 0 0>}
      localJointOffset=<4 0 0>
     }]
    simSpaceSettings {maxAngularAcceleration=4, maxAngularVelocity=400, simSpaceAngularAlpha=1}
   }
   n4 = Anim Dynamics - Bone: dyn_skirt_front_03(ComponentPose=@n3, Alpha=1, ExternalForce="0,0,0") {
    angularDampingOverride=0.999
    angularSpringConstant=20
    bAngularSpring=true
    bOverrideAngularDamping=true
    bUseGravityOverride=true
    gravityScale=0.35
    simulationSpace=RootRelative
    boundBone {boneName=dyn_skirt_front_03}
    componentAppliedLinearAccClamp=<0 0 0>
    gravityOverride {z=-1800}
    physicsBodyDefinitions [{
      boundBone {boneName=dyn_skirt_front_03}
      boxExtents=<15 15 30>
      constraintSetup {angularLimitsMax=<8 8 0>, angularLimitsMin=<-8 -8 -18>, angularTarget=<1 0 0>}
      localJointOffset=<4 0 0>
     }]
    simSpaceSettings {maxAngularAcceleration=4, maxAngularVelocity=400, simSpaceAngularAlpha=1}
   }
   n5 = Anim Dynamics - Bone: dyn_skirt_front_l_01(ComponentPose=@n4, Alpha=1, ExternalForce="0,0,0") {
    angularDampingOverride=0.999
    angularSpringConstant=20
    bAngularSpring=true
    bOverrideAngularDamping=true
    bUseGravityOverride=true
    gravityScale=0.35
    simulationSpace=RootRelative
    boundBone {boneName=dyn_skirt_front_l_01}
    componentAppliedLinearAccClamp=<0 0 0>
    gravityOverride {z=-1800}
    physicsBodyDefinitions [{
      boundBone {boneName=dyn_skirt_front_l_01}
      boxExtents=<15 15 30>
      constraintSetup {angularLimitsMax=<8 8 0>, angularLimitsMin=<-8 -8 -18>, angularTarget=<1 0 0>}
      localJointOffset=<4 0 0>
     }]
    simSpaceSettings {maxAngularAcceleration=4, maxAngularVelocity=400, simSpaceAngularAlpha=1}
   }
   n6 = Anim Dynamics - Bone: dyn_skirt_front_l_02(ComponentPose=@n5, Alpha=1, ExternalForce="0,0,0") {
    angularDampingOverride=0.999
    angularSpringConstant=20
    bAngularSpring=true
    bOverrideAngularDamping=true
    bUseGravityOverride=true
    gravityScale=0.35
    simulationSpace=RootRelative
    boundBone {boneName=dyn_skirt_front_l_02}
    componentAppliedLinearAccClamp=<0 0 0>
    gravityOverride {z=-1800}
    physicsBodyDefinitions [{
      boundBone {boneName=dyn_skirt_front_l_02}
      boxExtents=<15 15 30>
      constraintSetup {angularLimitsMax=<8 8 0>, angularLimitsMin=<-8 -8 -18>, angularTarget=<1 0 0>}
      localJointOffset=<4 0 0>
     }]
    simSpaceSettings {maxAngularAcceleration=4, maxAngularVelocity=400, simSpaceAngularAlpha=1}
   }
   n7 = Anim Dynamics - Bone: dyn_skirt_front_l_03(ComponentPose=@n6, Alpha=1, ExternalForce="0,0,0") {
    angularDampingOverride=0.999
    angularSpringConstant=20
    bAngularSpring=true
    bOverrideAngularDamping=true
    bUseGravityOverride=true
    gravityScale=0.35
    simulationSpace=RootRelative
    boundBone {boneName=dyn_skirt_front_l_03}
    componentAppliedLinearAccClamp=<0 0 0>
    gravityOverride {z=-1800}
    physicsBodyDefinitions [{
      boundBone {boneName=dyn_skirt_front_l_03}
      boxExtents=<15 15 30>
      constraintSetup {angularLimitsMax=<8 8 0>, angularLimitsMin=<-8 -8 -18>, angularTarget=<1 0 0>}
      localJointOffset=<4 0 0>
     }]
    simSpaceSettings {maxAngularAcceleration=4, maxAngularVelocity=400, simSpaceAngularAlpha=1}
   }
   n8 = Anim Dynamics - Bone: dyn_skirt_side_l_01(ComponentPose=@n7, Alpha=1, ExternalForce="0,0,0") {
    angularDampingOverride=0.999
    angularSpringConstant=20
    bAngularSpring=true
    bOverrideAngularDamping=true
    bUseGravityOverride=true
    gravityScale=0.35
    simulationSpace=RootRelative
    boundBone {boneName=dyn_skirt_side_l_01}
    componentAppliedLinearAccClamp=<0 0 0>
    gravityOverride {z=-1800}
    physicsBodyDefinitions [{
      boundBone {boneName=dyn_skirt_side_l_01}
      boxExtents=<15 15 30>
      constraintSetup {angularLimitsMax=<8 8 0>, angularLimitsMin=<-8 -8 -18>, angularTarget=<1 0 0>}
      localJointOffset=<4 0 0>
     }]
    simSpaceSettings {maxAngularAcceleration=4, maxAngularVelocity=400, simSpaceAngularAlpha=1}
   }
   n9 = Anim Dynamics - Bone: dyn_skirt_side_l_02(ComponentPose=@n8, Alpha=1, ExternalForce="0,0,0") {
    angularDampingOverride=0.999
    angularSpringConstant=20
    bAngularSpring=true
    bOverrideAngularDamping=true
    bUseGravityOverride=true
    gravityScale=0.35
    simulationSpace=RootRelative
    boundBone {boneName=dyn_skirt_side_l_02}
    componentAppliedLinearAccClamp=<0 0 0>
    gravityOverride {z=-1800}
    physicsBodyDefinitions [{
      boundBone {boneName=dyn_skirt_side_l_02}
      boxExtents=<15 15 30>
      constraintSetup {angularLimitsMax=<8 8 0>, angularLimitsMin=<-8 -8 -18>, angularTarget=<1 0 0>}
      localJointOffset=<4 0 0>
     }]
    simSpaceSettings {maxAngularAcceleration=4, maxAngularVelocity=400, simSpaceAngularAlpha=1}
   }
   n10 = Anim Dynamics - Bone: dyn_skirt_side_l_03(ComponentPose=@n9, Alpha=1, ExternalForce="0,0,0") {
    angularDampingOverride=0.999
    angularSpringConstant=20
    bAngularSpring=true
    bOverrideAngularDamping=true
    bUseGravityOverride=true
    gravityScale=0.35
    simulationSpace=RootRelative
    boundBone {boneName=dyn_skirt_side_l_03}
    componentAppliedLinearAccClamp=<0 0 0>
    gravityOverride {z=-1800}
    physicsBodyDefinitions [{
      boundBone {boneName=dyn_skirt_side_l_03}
      boxExtents=<15 15 30>
      constraintSetup {angularLimitsMax=<8 8 0>, angularLimitsMin=<-8 -8 -18>, angularTarget=<1 0 0>}
      localJointOffset=<4 0 0>
     }]
    simSpaceSettings {maxAngularAcceleration=4, maxAngularVelocity=400, simSpaceAngularAlpha=1}
   }
   n11 = Anim Dynamics - Bone: dyn_skirt_rear_l_01(ComponentPose=@n10, Alpha=1, ExternalForce="0,0,0") {
    angularDampingOverride=0.999
    angularSpringConstant=20
    bAngularSpring=true
    bOverrideAngularDamping=true
    bUseGravityOverride=true
    gravityScale=0.35
    simulationSpace=RootRelative
    boundBone {boneName=dyn_skirt_rear_l_01}
    componentAppliedLinearAccClamp=<0 0 0>
    gravityOverride {z=-1800}
    physicsBodyDefinitions [{
      boundBone {boneName=dyn_skirt_rear_l_01}
      boxExtents=<15 15 30>
      constraintSetup {angularLimitsMax=<8 8 0>, angularLimitsMin=<-8 -8 -18>, angularTarget=<1 0 0>}
      localJointOffset=<4 0 0>
     }]
    simSpaceSettings {maxAngularAcceleration=4, maxAngularVelocity=400, simSpaceAngularAlpha=1}
   }
   n12 = Anim Dynamics - Bone: dyn_skirt_rear_l_02(ComponentPose=@n11, Alpha=1, ExternalForce="0,0,0") {
    angularDampingOverride=0.999
    angularSpringConstant=20
    bAngularSpring=true
    bOverrideAngularDamping=true
    bUseGravityOverride=true
    gravityScale=0.35
    simulationSpace=RootRelative
    boundBone {boneName=dyn_skirt_rear_l_02}
    componentAppliedLinearAccClamp=<0 0 0>
    gravityOverride {z=-1800}
    physicsBodyDefinitions [{
      boundBone {boneName=dyn_skirt_rear_l_02}
      boxExtents=<15 15 30>
      constraintSetup {angularLimitsMax=<8 8 0>, angularLimitsMin=<-8 -8 -18>, angularTarget=<1 0 0>}
      localJointOffset=<4 0 0>
     }]
    simSpaceSettings {maxAngularAcceleration=4, maxAngularVelocity=400, simSpaceAngularAlpha=1}
   }
   n13 = Anim Dynamics - Bone: dyn_skirt_rear_l_03(ComponentPose=@n12, Alpha=1, ExternalForce="0,0,0") {
    angularDampingOverride=0.999
    angularSpringConstant=20
    bAngularSpring=true
    bOverrideAngularDamping=true
    bUseGravityOverride=true
    gravityScale=0.35
    simulationSpace=RootRelative
    boundBone {boneName=dyn_skirt_rear_l_03}
    componentAppliedLinearAccClamp=<0 0 0>
    gravityOverride {z=-1800}
    physicsBodyDefinitions [{
      boundBone {boneName=dyn_skirt_rear_l_03}
      boxExtents=<15 15 30>
      constraintSetup {angularLimitsMax=<8 8 0>, angularLimitsMin=<-8 -8 -18>, angularTarget=<1 0 0>}
      localJointOffset=<4 0 0>
     }]
    simSpaceSettings {maxAngularAcceleration=4, maxAngularVelocity=400, simSpaceAngularAlpha=1}
   }
   n14 = Anim Dynamics - Bone: dyn_skirt_rear_01(ComponentPose=@n13, Alpha=1, ExternalForce="0,0,0") {
    angularDampingOverride=0.999
    angularSpringConstant=20
    bAngularSpring=true
    bOverrideAngularDamping=true
    bUseGravityOverride=true
    gravityScale=0.35
    simulationSpace=RootRelative
    boundBone {boneName=dyn_skirt_rear_01}
    componentAppliedLinearAccClamp=<0 0 0>
    gravityOverride {z=-1800}
    physicsBodyDefinitions [{
      boundBone {boneName=dyn_skirt_rear_01}
      boxExtents=<15 15 30>
      constraintSetup {angularLimitsMax=<8 8 0>, angularLimitsMin=<-8 -8 -18>, angularTarget=<1 0 0>}
      localJointOffset=<4 0 0>
     }]
    simSpaceSettings {maxAngularAcceleration=4, maxAngularVelocity=400, simSpaceAngularAlpha=1}
   }
   n15 = Anim Dynamics - Bone: dyn_skirt_rear_02(ComponentPose=@n14, Alpha=1, ExternalForce="0,0,0") {
    angularDampingOverride=0.999
    angularSpringConstant=20
    bAngularSpring=true
    bOverrideAngularDamping=true
    bUseGravityOverride=true
    gravityScale=0.35
    simulationSpace=RootRelative
    boundBone {boneName=dyn_skirt_rear_02}
    componentAppliedLinearAccClamp=<0 0 0>
    gravityOverride {z=-1800}
    physicsBodyDefinitions [{
      boundBone {boneName=dyn_skirt_rear_02}
      boxExtents=<15 15 30>
      constraintSetup {angularLimitsMax=<8 8 0>, angularLimitsMin=<-8 -8 -18>, angularTarget=<1 0 0>}
      localJointOffset=<4 0 0>
     }]
    simSpaceSettings {maxAngularAcceleration=4, maxAngularVelocity=400, simSpaceAngularAlpha=1}
   }
   n16 = Anim Dynamics - Bone: dyn_skirt_rear_03(ComponentPose=@n15, Alpha=1, ExternalForce="0,0,0") {
    angularDampingOverride=0.999
    angularSpringConstant=20
    bAngularSpring=true
    bOverrideAngularDamping=true
    bUseGravityOverride=true
    gravityScale=0.35
    simulationSpace=RootRelative
    boundBone {boneName=dyn_skirt_rear_03}
    componentAppliedLinearAccClamp=<0 0 0>
    gravityOverride {z=-1800}
    physicsBodyDefinitions [{
      boundBone {boneName=dyn_skirt_rear_03}
      boxExtents=<15 15 30>
      constraintSetup {angularLimitsMax=<8 8 0>, angularLimitsMin=<-8 -8 -18>, angularTarget=<1 0 0>}
      localJointOffset=<4 0 0>
     }]
    simSpaceSettings {maxAngularAcceleration=4, maxAngularVelocity=400, simSpaceAngularAlpha=1}
   }
   n17 = Anim Dynamics - Bone: dyn_skirt_front_r_01(ComponentPose=@n16, Alpha=1, ExternalForce="0,0,0") {
    angularDampingOverride=0.999
    angularSpringConstant=20
    bAngularSpring=true
    bOverrideAngularDamping=true
    bUseGravityOverride=true
    gravityScale=0.35
    simulationSpace=RootRelative
    boundBone {boneName=dyn_skirt_front_r_01}
    componentAppliedLinearAccClamp=<0 0 0>
    gravityOverride {z=-1800}
    physicsBodyDefinitions [{
      boundBone {boneName=dyn_skirt_front_r_01}
      boxExtents=<15 15 30>
      constraintSetup {angularLimitsMax=<8 8 0>, angularLimitsMin=<-8 -8 -18>, angularTarget=<1 0 0>}
      localJointOffset=<-4 0 0>
     }]
    simSpaceSettings {maxAngularAcceleration=4, maxAngularVelocity=400, simSpaceAngularAlpha=1}
   }
   n18 = Anim Dynamics - Bone: dyn_skirt_front_r_02(ComponentPose=@n17, Alpha=1, ExternalForce="0,0,0") {
    angularDampingOverride=0.999
    angularSpringConstant=20
    bAngularSpring=true
    bOverrideAngularDamping=true
    bUseGravityOverride=true
    gravityScale=0.35
    simulationSpace=RootRelative
    boundBone {boneName=dyn_skirt_front_r_02}
    componentAppliedLinearAccClamp=<0 0 0>
    gravityOverride {z=-1800}
    physicsBodyDefinitions [{
      boundBone {boneName=dyn_skirt_front_r_02}
      boxExtents=<15 15 30>
      constraintSetup {angularLimitsMax=<8 8 0>, angularLimitsMin=<-8 -8 -18>, angularTarget=<1 0 0>}
      localJointOffset=<-4 0 0>
     }]
    simSpaceSettings {maxAngularAcceleration=4, maxAngularVelocity=400, simSpaceAngularAlpha=1}
   }
   n19 = Anim Dynamics - Bone: dyn_skirt_front_r_03(ComponentPose=@n18, Alpha=1, ExternalForce="0,0,0") {
    angularDampingOverride=0.999
    angularSpringConstant=20
    bAngularSpring=true
    bOverrideAngularDamping=true
    bUseGravityOverride=true
    gravityScale=0.35
    simulationSpace=RootRelative
    boundBone {boneName=dyn_skirt_front_r_03}
    componentAppliedLinearAccClamp=<0 0 0>
    gravityOverride {z=-1800}
    physicsBodyDefinitions [{
      boundBone {boneName=dyn_skirt_front_r_03}
      boxExtents=<15 15 30>
      constraintSetup {angularLimitsMax=<8 8 0>, angularLimitsMin=<-8 -8 -18>, angularTarget=<1 0 0>}
      localJointOffset=<-4 0 0>
     }]
    simSpaceSettings {maxAngularAcceleration=4, maxAngularVelocity=400, simSpaceAngularAlpha=1}
   }
   n20 = Anim Dynamics - Bone: dyn_skirt_side_r_01(ComponentPose=@n19, Alpha=1, ExternalForce="0,0,0") {
    angularDampingOverride=0.999
    angularSpringConstant=20
    bAngularSpring=true
    bOverrideAngularDamping=true
    bUseGravityOverride=true
    gravityScale=0.35
    simulationSpace=RootRelative
    boundBone {boneName=dyn_skirt_side_r_01}
    componentAppliedLinearAccClamp=<0 0 0>
    gravityOverride {z=-1800}
    physicsBodyDefinitions [{
      boundBone {boneName=dyn_skirt_side_r_01}
      boxExtents=<15 15 30>
      constraintSetup {angularLimitsMax=<8 8 0>, angularLimitsMin=<-8 -8 -18>, angularTarget=<1 0 0>}
      localJointOffset=<-4 0 0>
     }]
    simSpaceSettings {maxAngularAcceleration=4, maxAngularVelocity=400, simSpaceAngularAlpha=1}
   }
   n21 = Anim Dynamics - Bone: dyn_skirt_side_r_02(ComponentPose=@n20, Alpha=1, ExternalForce="0,0,0") {
    angularDampingOverride=0.999
    angularSpringConstant=20
    bAngularSpring=true
    bOverrideAngularDamping=true
    bUseGravityOverride=true
    gravityScale=0.35
    simulationSpace=RootRelative
    boundBone {boneName=dyn_skirt_side_r_02}
    componentAppliedLinearAccClamp=<0 0 0>
    gravityOverride {z=-1800}
    physicsBodyDefinitions [{
      boundBone {boneName=dyn_skirt_side_r_02}
      boxExtents=<15 15 30>
      constraintSetup {angularLimitsMax=<8 8 0>, angularLimitsMin=<-8 -8 -18>, angularTarget=<1 0 0>}
      localJointOffset=<-4 0 0>
     }]
    simSpaceSettings {maxAngularAcceleration=4, maxAngularVelocity=400, simSpaceAngularAlpha=1}
   }
   n22 = Anim Dynamics - Bone: dyn_skirt_side_r_03(ComponentPose=@n21, Alpha=1, ExternalForce="0,0,0") {
    angularDampingOverride=0.999
    angularSpringConstant=20
    bAngularSpring=true
    bOverrideAngularDamping=true
    bUseGravityOverride=true
    gravityScale=0.35
    simulationSpace=RootRelative
    boundBone {boneName=dyn_skirt_side_r_03}
    componentAppliedLinearAccClamp=<0 0 0>
    gravityOverride {z=-1800}
    physicsBodyDefinitions [{
      boundBone {boneName=dyn_skirt_side_r_03}
      boxExtents=<15 15 30>
      constraintSetup {angularLimitsMax=<8 8 0>, angularLimitsMin=<-8 -8 -18>, angularTarget=<1 0 0>}
      localJointOffset=<-4 0 0>
     }]
    simSpaceSettings {maxAngularAcceleration=4, maxAngularVelocity=400, simSpaceAngularAlpha=1}
   }
   n23 = Anim Dynamics - Bone: dyn_skirt_rear_r_01(ComponentPose=@n22, Alpha=1, ExternalForce="0,0,0") {
    angularDampingOverride=0.999
    angularSpringConstant=20
    bAngularSpring=true
    bOverrideAngularDamping=true
    bUseGravityOverride=true
    gravityScale=0.35
    simulationSpace=RootRelative
    boundBone {boneName=dyn_skirt_rear_r_01}
    componentAppliedLinearAccClamp=<0 0 0>
    gravityOverride {z=-1800}
    physicsBodyDefinitions [{
      boundBone {boneName=dyn_skirt_rear_r_01}
      boxExtents=<15 15 30>
      constraintSetup {angularLimitsMax=<8 8 0>, angularLimitsMin=<-8 -8 -18>, angularTarget=<1 0 0>}
      localJointOffset=<-4 0 0>
     }]
    simSpaceSettings {maxAngularAcceleration=4, maxAngularVelocity=400, simSpaceAngularAlpha=1}
   }
   n24 = Anim Dynamics - Bone: dyn_skirt_rear_r_02(ComponentPose=@n23, Alpha=1, ExternalForce="0,0,0") {
    angularDampingOverride=0.999
    angularSpringConstant=20
    bAngularSpring=true
    bOverrideAngularDamping=true
    bUseGravityOverride=true
    gravityScale=0.35
    simulationSpace=RootRelative
    boundBone {boneName=dyn_skirt_rear_r_02}
    componentAppliedLinearAccClamp=<0 0 0>
    gravityOverride {z=-1800}
    physicsBodyDefinitions [{
      boundBone {boneName=dyn_skirt_rear_r_02}
      boxExtents=<15 15 30>
      constraintSetup {angularLimitsMax=<8 8 0>, angularLimitsMin=<-8 -8 -18>, angularTarget=<1 0 0>}
      localJointOffset=<-4 0 0>
     }]
    simSpaceSettings {maxAngularAcceleration=4, maxAngularVelocity=400, simSpaceAngularAlpha=1}
   }
   n25 = Anim Dynamics - Bone: dyn_skirt_rear_r_03(ComponentPose=@n24, Alpha=1, ExternalForce="0,0,0") {
    angularDampingOverride=0.999
    angularSpringConstant=20
    bAngularSpring=true
    bOverrideAngularDamping=true
    bUseGravityOverride=true
    gravityScale=0.35
    simulationSpace=RootRelative
    boundBone {boneName=dyn_skirt_rear_r_03}
    componentAppliedLinearAccClamp=<0 0 0>
    gravityOverride {z=-1800}
    physicsBodyDefinitions [{
      boundBone {boneName=dyn_skirt_rear_r_03}
      boxExtents=<15 15 30>
      constraintSetup {angularLimitsMax=<8 8 0>, angularLimitsMin=<-8 -8 -18>, angularTarget=<1 0 0>}
      localJointOffset=<-4 0 0>
     }]
    simSpaceSettings {maxAngularAcceleration=4, maxAngularVelocity=400, simSpaceAngularAlpha=1}
   }
   Save cached pose 'PostAnimDynSkirt'(Pose=Component To Local(ComponentPose=@n25)) {cacheName=PostAnimDynSkirt}
  ]}
 EventGraph {nodes [
   n1 = Event BlueprintUpdateAnimation()
   > n2 = "Cloth Controller Function (Parameter Asset) (Nativized)"!(
    execute
    CharacterPartAnimInstance:Custom Character Part Anim Instance Object Reference=Self-Reference()
    ClothingAssetName:Name=F_MED_VowelSleep_Clothing
    ClothCoreSettings:"Cloth Core Settings Structure (by ref)"=Get Cloth Core Settings() -> (Cloth Core Settings:Cloth Core Settings Structure)
    WindGustLocalSettings:"Wind Gust Settings Structure (by ref)"=Get Wind Gust Local Settings Out() -> (Wind Gust Local Settings Out:Wind Gust Settings Structure)
    ClothUpAxis:Vector="0,0,1"
    LinVelScaleFwdBwdClamp:float=0
    LinVelScaleUpDwnClamp:float=0.7
    DeltaTime:float=@n1.DeltaTimeX
    GameTimeInSeconds:float=GetGameTimeInSeconds()
    StateLogicLayerName:Name=none
    Parameters:Fort Character Dynamics Parameters Object Reference=&/F_MED_VowelSleep_Parameters
   ) -> (
    then
    OutCurrentState:Name=None
    OutClothParameters:Cloth Parameters Structure
    OutClothCoreSettings:Cloth Core Settings Structure
    OutWindGustSettings:Wind Gust Parameters Structure
    OutWindGustLocalSettings:Wind Gust Settings Structure
    OutWindVelocity:Vector
    OutLinearVelVector:Vector
   )
   > Set Wind Gust Local Settings Out(Wind Gust Local Settings Out:Wind Gust Settings Structure=@n2.OutWindGustLocalSettings) -> (Output_Get:Wind Gust Settings Structure)
   Event OnEndKneeSliding() {is=title K2Node_Event}
   > Set DelayedKneeSlide(DelayedKneeSlide:bool=false) -> (Output_Get:bool)
   n3 = Event OnBeginCrouching()
   n4 = Get!() -> (
    Default Cloth Core Settings_EdgeStiffness:Vector 2D Structure
    Default Cloth Core Settings_BendingStiffness:Vector 2D Structure
    Default Cloth Core Settings_AreaStiffness:Vector 2D Structure
    Default Cloth Core Settings_TetherStiffness:Vector 2D Structure
    Default Cloth Core Settings_CollisionThickness:float=~
    Default Cloth Core Settings_FrictionCoefficient:float=~
    Default Cloth Core Settings_SelfCollisionThickness:float=~
    Default Cloth Core Settings_DampingCoefficient:float=~
    Default Cloth Core Settings_LocalDampingCoefficient:float=~
    Default Cloth Core Settings_AnimDriveStiffness:Vector 2D Structure
    Default Cloth Core Settings_AnimDriveDamping:Vector 2D Structure
    Default Cloth Core Settings_NumberIterations:int
    Default Cloth Core Settings_NumberSubsteps:int
   )
   Set!(
    execute=[@n3.then, @n5.then]
    Cloth Core Settings_EdgeStiffness:Vector 2D Structure=@n4
    Cloth Core Settings_BendingStiffness:Vector 2D Structure=@n4.Default Cloth Core Settings_BendingStiffness
    Cloth Core Settings_AreaStiffness:Vector 2D Structure=@n4.Default Cloth Core Settings_AreaStiffness
    Cloth Core Settings_TetherStiffness:Vector 2D Structure=@n4.Default Cloth Core Settings_TetherStiffness
    Cloth Core Settings_CollisionThickness:float=@n4.Default Cloth Core Settings_CollisionThickness
    Cloth Core Settings_FrictionCoefficient:float=@n4.Default Cloth Core Settings_FrictionCoefficient
    Cloth Core Settings_SelfCollisionThickness:float=@n4.Default Cloth Core Settings_SelfCollisionThickness
    Cloth Core Settings_DampingCoefficient:float=@n4.Default Cloth Core Settings_DampingCoefficient
    Cloth Core Settings_LocalDampingCoefficient:float=@n4.Default Cloth Core Settings_LocalDampingCoefficient
    Cloth Core Settings_AnimDriveStiffness:Vector 2D Structure=(X=0.4,Y=1)
    Cloth Core Settings_AnimDriveDamping:Vector 2D Structure=(X=0.4,Y=1)
    Cloth Core Settings_NumberIterations:int=@n4.Default Cloth Core Settings_NumberIterations
    Cloth Core Settings_NumberSubsteps:int=@n4.Default Cloth Core Settings_NumberSubsteps
   ) -> (then, Output_Get:Cloth Core Settings Structure)
   > Delay(Duration=0.2)
   > Set Cloth Core Settings(Cloth Core Settings:Cloth Core Settings Structure=Get Default Cloth Core Settings() -> (Default Cloth Core Settings:Cloth Core Settings Structure)) -> (Output_Get:Cloth Core Settings Structure)
   n5 = Event OnEndCrouching()
   Event OnBeginKneeSliding() {is=title K2Node_Event}
   > Delay(Duration=0.2)
   > Set DelayedKneeSlide(DelayedKneeSlide=true)
  ]}
]
variables [
 Cloth Core Settings {
  type=ClothCoreSettings
  default {
   collisionThickness=1
   dampingCoefficient=0
   frictionCoefficient=0.8
   localDampingCoefficient=0.08
   numberIterations=2
   numberSubsteps=1
   selfCollisionThickness=0
   animDriveDamping=<0.1 1>
   animDriveStiffness=<0.1 1>
   areaStiffness=<1 1>
   bendingStiffness=<0.85 1>
   edgeStiffness=<0.85 1>
   tetherStiffness=<1.05 1.05>
  }
 }
 Default Cloth Core Settings {
  type=ClothCoreSettings
  default {
   collisionThickness=1.5
   dampingCoefficient=0.1
   frictionCoefficient=0.8
   localDampingCoefficient=0
   numberIterations=2
   numberSubsteps=1
   selfCollisionThickness=0
   animDriveDamping=<0.1 1>
   animDriveStiffness=<0.1 1>
   areaStiffness=<1 1>
   bendingStiffness=<0.7 1>
   edgeStiffness=<0.7 1>
   tetherStiffness=<1.05 1.05>
  }
 }
 DelayedKneeSlide {type=boolean}
 Output {type=SkeletalMeshComponentBudgeted}
 Wind Gust Local Settings {type=WindGustSettings, default {elapsedTime=0, finalGustStrengthLocal=0, gustDuration=0, gustStrength=0, timeBetweenGusts=0}}
 Wind Gust Local Settings Out {type=WindGustSettings, default {elapsedTime=0, finalGustStrengthLocal=0, gustDuration=0, gustStrength=0, timeBetweenGusts=0}}
]
