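# Physics asset for the F_MED_VowelSleep female body that, as its name indicates, supplies collision shapes for the character's cape.
# Capsules cover the pelvis, spine chain, clavicles and upper neck, with one sphere for the head.
# Arm, leg and foot bodies are present but carry no shapes.
# It has no constraints, so it serves as a collider set rather than a jointed ragdoll.
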PhysicsAsset /BRCosmetics/Characters/Player/Female/Medium/Bodies/F_MED_VowelSleep/Meshes/F_MED_VowelSleep_CapeColliders
bodies {
 calf_l []
 calf_r []
 clavicle_l [{shapeType=Capsule, length=8.4765, radius=6.0892, center=<5.455 -1.5056 -1.1198>, rotation=<rot 83.3176 -4.9345 170.6982>}]
 clavicle_r [{shapeType=Capsule, length=8.4765, radius=6.0892, center=<-5.455 1.5056 1.1198>, rotation=<rot 83.3176 -4.9345 -9.3018>}]
 foot_l []
 foot_r []
 head [{shapeType=Sphere, radius=17.3125, center=<11.028 4.8973 0>}]
 lowerarm_l []
 lowerarm_r []
 neck_02 [{shapeType=Capsule, length=7.7866, radius=7.0156, center=<-1.495 3.6586 0>, rotation=<rot 90 4.2785 0>}]
 pelvis [
  {shapeType=Capsule, length=20.4217, radius=14.9115, center=<-13.2721 -4.5891 0>}
  {shapeType=Capsule, length=8.4504, radius=15, shapeName=pelvis_capsule}
 ]
 spine_01 [{shapeType=Capsule, length=9.7497, radius=9.2108, center=<-1.7965 0.1884 0>}]
 spine_02 [
  {shapeType=Capsule, length=6.2037, radius=9.4443, center=<-1.711 1.2936 0.0001>}
  {shapeType=Capsule, length=29.3956, radius=11.5807, center=<2.5024 2.3084 3.9825>, rotation=<rot 90 0 0>}
  {shapeType=Capsule, length=31.3885, radius=9.5877, center=<2.4704 0.4991 -3.9825>, rotation=<rot 90 -0.0396 -0.0396>}
 ]
 spine_03 [{shapeType=Capsule, length=8.4663, radius=8.471, center=<0.1445 0.2396 0.0001>}]
 spine_04 [{shapeType=Capsule, length=4.6298, radius=11.8232, center=<0.9397 1.2948 0.0002>}]
 spine_05 [
  {shapeType=Capsule, length=7.7517, radius=11.9844, center=<-4.0281 1.2406 0>}
  {shapeType=Capsule, length=12.8499, radius=10.1327, center=<-0.2916 0.4452 0>}
 ]
 thigh_l []
 thigh_r []
 upperarm_l []
 upperarm_r []
}
constraints []
