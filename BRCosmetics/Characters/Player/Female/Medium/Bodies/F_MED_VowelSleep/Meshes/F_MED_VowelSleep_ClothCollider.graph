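# Cloth collision physics asset for the VowelSleep medium female player body cosmetic.
# Capsules approximate the lower body, with one on each calf and foot, mirrored left to right, and two on the pelvis.
# Both thigh bodies are present but hold no shapes.
# No constraints are defined, so the bodies are not jointed as a ragdoll.
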PhysicsAsset /BRCosmetics/Characters/Player/Female/Medium/Bodies/F_MED_VowelSleep/Meshes/F_MED_VowelSleep_ClothCollider
bodies {
 calf_l [{
   shapeType=Capsule
   length=28.2043
   radius=7.6238
   shapeName=calf_l_capsule
   center=<-33.8336 -2.9225 -0.6437>
   rotation=<rot -89.6147 -179.9999 176.0082>
  }]
 calf_r [{
   shapeType=Capsule
   length=28.2043
   radius=7.6238
   shapeName=calf_r_capsule
   center=<33.8336 2.9225 0.6437>
   rotation=<rot -89.6147 -179.9999 -3.9918>
  }]
 foot_l [{
   shapeType=Capsule
   length=7.2902
   radius=9.0504
   shapeName=foot_l_capsule
   center=<-1.9446 6.2339 -0.1526>
   rotation=<rot 0.6166 -2.3616 -87.7705>
  }]
 foot_r [{
   shapeType=Capsule
   length=7.2902
   radius=9.0504
   shapeName=foot_r_capsule
   center=<1.9446 -6.2339 0.1526>
   rotation=<rot 0.6166 -2.3616 92.2295>
  }]
 pelvis [
  {shapeType=Capsule, length=10.671, radius=9.8427, center=<-5.768 -1.1533 0>}
  {shapeType=Capsule, length=14.7511, radius=5.7174, shapeName=pelvis_capsule, center=<0 -1.2975 0>}
 ]
 thigh_l []
 thigh_r []
}
constraints []
