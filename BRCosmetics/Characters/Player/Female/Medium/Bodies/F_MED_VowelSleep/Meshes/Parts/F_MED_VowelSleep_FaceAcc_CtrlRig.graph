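# Control Rig for the F_MED_VowelSleep face accessory part that poses the character's long hair bones around the body and worn back items.
# A branch on backbling worn without a cape picks the rear hair setup: bases aimed at spine_04 for full simulation, or strands constrained to head, neck, spine and clavicles for backbling.
# Spherical pose readers on clavicles and upper arms weight corrective offsets for hair bones, with one right clavicle reader smoothed by a Kalman filter.
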ControlRigBlueprint /BRCosmetics/Characters/Player/Female/Medium/Bodies/F_MED_VowelSleep/Meshes/Parts/F_MED_VowelSleep_FaceAcc_CtrlRig
graphs [
 CollapseNode_ContainedGraph {nodes [
   n1 = RigVMFunctionEntryNode()
   &RigUnit_AimConstraintLocalSpaceOffset(
    Child=(Type=Bone,Name="dyn_hair_rear_l_base")
    Filter=(bX=True,bY=True,bZ=True)
    Parents=((Item=(Type=Bone,Name="spine_04"),Weight=1))
    Weight=1
    WorldUp=(Target=(X=0,Y=0,Z=1),Kind=Location,Space=(Type=Bone,Name="spine_04"))
    ExecutePin=@n1.ExecutePin
   )
   > &RigUnit_AimConstraintLocalSpaceOffset(
    Child=(Type=Bone,Name="dyn_hair_rear_base")
    Filter=(bX=True,bY=True,bZ=True)
    Parents=((Item=(Type=Bone,Name="spine_04"),Weight=1))
    Weight=1
    WorldUp=(Target=(X=0,Y=0,Z=1),Kind=Location,Space=(Type=Bone,Name="spine_04"))
   )
   > n2 = &RigUnit_AimConstraintLocalSpaceOffset(
    AimAxis=(X=-1,Y=0,Z=0)
    Child=(Type=Bone,Name="dyn_hair_rear_r_base")
    Filter=(bX=True,bY=True,bZ=True)
    Parents=((Item=(Type=Bone,Name="spine_04"),Weight=1))
    Weight=1
    WorldUp=(Target=(X=0,Y=0,Z=1),Kind=Location,Space=(Type=Bone,Name="spine_04"))
   )
   RigVMFunctionReturnNode(ExecutePin=@n2.ExecutePin)
  ]}
 CollapseNode_ContainedGraph {nodes [
   n1 = RigVMFunctionEntryNode()
   RigVMFunction_Sequence = RigVMAggregateNode(ExecuteContext=@n1.ExecutePin)
   &RigUnit_ParentConstraint(
    AdvancedSettings=(InterpolationType=Average,RotationOrderForFilter=XZY)
    Child=(Type=Bone,Name="dyn_hair_rear_l_02")
    Parents=((Item=(Type=Bone,Name="head"),Weight=1))
    Weight=1
    ExecutePin=@RigVMFunction_Sequence.A
   )
   > &RigUnit_ParentConstraint(
    AdvancedSettings=(InterpolationType=Average,RotationOrderForFilter=XZY)
    Child=(Type=Bone,Name="dyn_hair_rear_l_03")
    Parents=((Item=(Type=Bone,Name="head"),Weight=0.034),(Item=(Type=Bone,Name="neck_02"),Weight=0.175),(Item=(Type=Bone,Name="neck_01"),Weight=0.145),(Item=(Type=Bone,Name="spine_05"),Weight=0.6),(Item=(Type=Bone,Name="clavicle_l"),Weight=0.02))
    Weight=1
   )
   > &RigUnit_ParentConstraint(
    AdvancedSettings=(InterpolationType=Average,RotationOrderForFilter=XZY)
    Child=(Type=Bone,Name="dyn_hair_rear_l_04")
    Parents=((Item=(Type=Bone,Name="spine_05"),Weight=0.421),(Item=(Type=Bone,Name="clavicle_l"),Weight=0.579))
    Weight=1
   )
   > &RigUnit_ParentConstraint(
    AdvancedSettings=(InterpolationType=Average,RotationOrderForFilter=XZY)
    Child=(Type=Bone,Name="dyn_hair_rear_02")
    Parents=((Item=(Type=Bone,Name="head"),Weight=1))
    Weight=1
   )
   > &RigUnit_ParentConstraint(
    AdvancedSettings=(InterpolationType=Average,RotationOrderForFilter=XZY)
    Child=(Type=Bone,Name="dyn_hair_rear_03")
    Parents=((Item=(Type=Bone,Name="head"),Weight=0.028),(Item=(Type=Bone,Name="neck_02"),Weight=0.161),(Item=(Type=Bone,Name="neck_01"),Weight=0.14),(Item=(Type=Bone,Name="spine_05"),Weight=0.626),(Item=(Type=Bone,Name="spine_03"),Weight=0.044))
    Weight=1
   )
   > &RigUnit_ParentConstraint(
    AdvancedSettings=(InterpolationType=Average,RotationOrderForFilter=XZY)
    Child=(Type=Bone,Name="dyn_hair_rear_04")
    Parents=((Item=(Type=Bone,Name="spine_05"),Weight=0.8),(Item=(Type=Bone,Name="clavicle_l"),Weight=0.1),(Item=(Type=Bone,Name="clavicle_r"),Weight=0.1))
    Weight=1
   )
   > &RigUnit_ParentConstraint(
    AdvancedSettings=(InterpolationType=Average,RotationOrderForFilter=XZY)
    Child=(Type=Bone,Name="dyn_hair_rear_05")
    Parents=((Item=(Type=Bone,Name="spine_05"),Weight=0.6),(Item=(Type=Bone,Name="spine_04"),Weight=0.2),(Item=(Type=Bone,Name="clavicle_l"),Weight=0.1),(Item=(Type=Bone,Name="clavicle_r"),Weight=0.1))
    Weight=1
   )
   > &RigUnit_ParentConstraint(
    AdvancedSettings=(InterpolationType=Average,RotationOrderForFilter=XZY)
    Child=(Type=Bone,Name="dyn_hair_rear_r_02")
    Parents=((Item=(Type=Bone,Name="head"),Weight=1))
    Weight=1
   )
   > &RigUnit_ParentConstraint(
    AdvancedSettings=(InterpolationType=Average,RotationOrderForFilter=XZY)
    Child=(Type=Bone,Name="dyn_hair_rear_r_03")
    Parents=((Item=(Type=Bone,Name="head"),Weight=0.034),(Item=(Type=Bone,Name="neck_02"),Weight=0.175),(Item=(Type=Bone,Name="neck_01"),Weight=0.145),(Item=(Type=Bone,Name="spine_05"),Weight=0.6),(Item=(Type=Bone,Name="clavicle_r"),Weight=0.02))
    Weight=1
   )
   > &RigUnit_ParentConstraint(
    AdvancedSettings=(InterpolationType=Average,RotationOrderForFilter=XZY)
    Child=(Type=Bone,Name="dyn_hair_rear_r_04")
    Parents=((Item=(Type=Bone,Name="spine_05"),Weight=0.421),(Item=(Type=Bone,Name="clavicle_r"),Weight=0.579))
    Weight=1
   )
   &RigUnit_ModifyTransforms(
    ItemToModify="((Item=(Type=Bone,Name=\"dyn_hair_rear_02\"),Transform=(Rotation=(X=0,Y=0,Z=0,W=-1),Translation=(X=0,Y=0.4289,Z=0),Scale3D=(X=1,Y=1,Z=1))),(Item=(Type=Bone,Name=\"dyn_hair_rear_03\"),Transform=(Rotation=(X=0,Y=0,Z=0.0125,W=-0.9999),Translation=(X=0.0301,Y=1.18 ...(truncated)"
    Weight=1
    WeightMinimum=0
    ExecutePin=@RigVMFunction_Sequence.B
   )
   RigVMFunctionReturnNode(ExecutePin=@RigVMFunction_Sequence.C)
  ]}
 RigVMFunctionLibrary {links [], nodes []}
 RigVMFunction_Sequence_1_ContainedGraph {nodes [
   n1 = RigVMFunctionEntryNode()
   RigVMFunction_Sequence = &RigVMFunction_Sequence(ExecuteContext=@n1.ExecuteContext)
   n2 = &RigVMFunction_Sequence(ExecuteContext=@RigVMFunction_Sequence.B)
   n3 = &RigVMFunction_Sequence(ExecuteContext=@n2.B)
   n4 = &RigVMFunction_Sequence(ExecuteContext=@n3.B)
   RigVMFunctionReturnNode(A=@RigVMFunction_Sequence.A, B=@n2.A, C=@n3.A, D=@n4.A, E=@n4.B)
  ]}
 RigVMFunction_Sequence_1_ContainedGraph {nodes [
   n1 = RigVMFunctionEntryNode()
   RigVMFunction_Sequence = &RigVMFunction_Sequence(ExecuteContext=@n1.ExecuteContext)
   n2 = &RigVMFunction_Sequence(ExecuteContext=@RigVMFunction_Sequence.B)
   RigVMFunctionReturnNode(A=@RigVMFunction_Sequence.A, B=@n2.A, C=@n2.B)
  ]}
 RigVMModel {nodes [
   RigUnit_BeginExecution = &RigUnit_BeginExecution()
   RigVMFunction_Sequence = RigVMAggregateNode(ExecuteContext=@RigUnit_BeginExecution.ExecutePin)
   n1 = &RigUnit_SphericalPoseReader(
    ActiveRegionScaleFactors=(PositiveWidth=1,NegativeWidth=2.4,PositiveHeight=1,NegativeHeight=1)
    ActiveRegionSize=0.1
    Debug=(bDrawDebug=false,bDraw2D=False,bDrawLocalAxes=false,DebugScale=25,DebugSegments=20,DebugThickness=0.25)
    DriverAxis=(X=1,Y=0,Z=0)
    DriverItem=(Type=Bone,Name="clavicle_l")
    FalloffRegionScaleFactors=(PositiveWidth=1,NegativeWidth=1,PositiveHeight=1,NegativeHeight=1)
    FalloffSize=0.24
    FlipHeightScaling=False
    FlipWidthScaling=False
    OptionalParentItem=(Type=Bone,Name="None")
    RotationOffset=(X=0,Y=-45,Z=0)
    ExecutePin=@RigVMFunction_Sequence.D
   )
   > &RigUnit_ModifyTransforms(
    ItemToModify="((Item=(Type=Bone,Name=\"deform_hair_split_l\"),Transform=(Rotation=(X=0,Y=0,Z=0,W=-1),Translation=(X=2.6141,Y=1.2308,Z=-2.0235),Scale3D=(X=1,Y=1,Z=0.3019))),(Item=(Type=Bone,Name=\"dyn_hair_rear_l_base\"),Transform=(Rotation=(X=-0.0001,Y=0.0065,Z=0.0217,W=-0. ...(truncated)"
    WeightMinimum=0
    Weight=@n1.OutputParam
   )
   n2 = &RigUnit_SphericalPoseReader(
    ActiveRegionScaleFactors=(PositiveWidth=1,NegativeWidth=2.4,PositiveHeight=1,NegativeHeight=1)
    ActiveRegionSize=0.1
    Debug=(bDrawDebug=false,bDraw2D=False,bDrawLocalAxes=false,DebugScale=25,DebugSegments=20,DebugThickness=0.25)
    DriverAxis=(X=-1,Y=0,Z=0)
    DriverItem=(Type=Bone,Name="clavicle_r")
    FalloffRegionScaleFactors=(PositiveWidth=1,NegativeWidth=1,PositiveHeight=1,NegativeHeight=1)
    FalloffSize=0.24
    FlipHeightScaling=False
    FlipWidthScaling=False
    OptionalParentItem=(Type=Bone,Name="None")
    RotationOffset=(X=0,Y=135,Z=0)
    ExecutePin=@RigVMFunction_Sequence.E
   )
   > &RigUnit_ModifyTransforms(
    ItemToModify="((Item=(Type=Bone,Name=\"deform_hair_split_r\"),Transform=(Rotation=(X=0,Y=0,Z=0,W=1),Translation=(X=-4.6894,Y=-1.1196,Z=3.1499),Scale3D=(X=1,Y=1,Z=0.5679))),(Item=(Type=Bone,Name=\"dyn_hair_rear_r_base\"),Transform=(Rotation=(X=0,Y=0,Z=0.0087,W=1),Translation ...(truncated)"
    WeightMinimum=0
    Weight=@n2.OutputParam
   )
   > n3 = &RigUnit_SphericalPoseReader(
    ActiveRegionScaleFactors=(PositiveWidth=1,NegativeWidth=2,PositiveHeight=2,NegativeHeight=1)
    ActiveRegionSize=0.1
    Debug=(bDrawDebug=false,bDraw2D=False,bDrawLocalAxes=false,DebugScale=25,DebugSegments=20,DebugThickness=0.25)
    DriverAxis=(X=-1,Y=0,Z=0)
    DriverItem=(Type=Bone,Name="clavicle_r")
    FalloffRegionScaleFactors=(PositiveWidth=0.86,NegativeWidth=1,PositiveHeight=1,NegativeHeight=1)
    FalloffSize=0.2
    FlipHeightScaling=False
    FlipWidthScaling=False
    OptionalParentItem=(Type=Bone,Name="None")
    RotationOffset=(X=-135.9317,Y=51.3295,Z=-150.255)
   )
   KalmanFilter = &RigVMFunction_KalmanFloat(BufferSize=8, Value=@n3.OutputParam) {template="KalmanFilter::Execute(in Value,in BufferSize,out Result)"}
   > &RigUnit_ModifyTransforms(
    ItemToModify="((Item=(Type=Bone,Name=\"deform_hair_split_r\"),Transform=(Rotation=(X=-0.3011,Y=0.261,Z=0.0411,W=-0.9162),Translation=(X=-5.4292,Y=-1.861,Z=-7.0465),Scale3D=(X=1,Y=0.2,Z=1))),(Item=(Type=Bone,Name=\"dyn_hair_rear_r_base\"),Transform=(Rotation=(X=0.0411,Y=-0.0 ...(truncated)"
    WeightMinimum=0
    Weight=@KalmanFilter.Result
   )
   RigVMFunction_MathBoolNot = &RigVMFunction_MathBoolNot(Value=RigVMVariableNode() {variable=Is Wearing Cape}.Value)
   RigVMFunction_MathBoolAnd = &RigVMFunction_MathBoolAnd(B=@RigVMFunction_MathBoolNot.Result, A=RigVMVariableNode() {variable=Is Wearing Backbling}.Value)
   RigVMFunction_ControlFlowBranch = &RigVMFunction_ControlFlowBranch(Condition=@RigVMFunction_MathBoolAnd.Result, ExecuteContext=@RigVMFunction_Sequence.A)
   Rear Hair Posing - Full Simulation = RigVMCollapseNode(ExecutePin=@RigVMFunction_ControlFlowBranch.False)
   Rear Hair Posing - Backbling Simulation = RigVMCollapseNode(ExecutePin=@RigVMFunction_ControlFlowBranch.True)
   n4 = &RigUnit_SphericalPoseReader(
    ActiveRegionScaleFactors=(PositiveWidth=1,NegativeWidth=3,PositiveHeight=1.5,NegativeHeight=3)
    ActiveRegionSize=0.1
    Debug=(bDrawDebug=false,bDraw2D=False,bDrawLocalAxes=False,DebugScale=25,DebugSegments=20,DebugThickness=0.25)
    DriverAxis=(X=-1,Y=0,Z=0)
    DriverItem=(Type=Bone,Name="upperarm_r")
    FalloffRegionScaleFactors=(PositiveWidth=1,NegativeWidth=1,PositiveHeight=1,NegativeHeight=1)
    FalloffSize=0.4
    FlipHeightScaling=False
    FlipWidthScaling=False
    OptionalParentItem=(Type=Bone,Name="clavicle_r")
    RotationOffset=(X=169.9189,Y=63.9929,Z=-78.8212)
    ExecutePin=@RigVMFunction_Sequence.C
   )
   > n5 = &RigUnit_SphericalPoseReader(
    ActiveRegionScaleFactors=(PositiveWidth=1,NegativeWidth=3,PositiveHeight=1.5,NegativeHeight=3)
    ActiveRegionSize=0.1
    Debug=(bDrawDebug=false,bDraw2D=False,bDrawLocalAxes=False,DebugScale=25,DebugSegments=20,DebugThickness=0.25)
    DriverAxis=(X=-1,Y=0,Z=0)
    DriverItem=(Type=Bone,Name="upperarm_r")
    FalloffRegionScaleFactors=(PositiveWidth=1,NegativeWidth=1,PositiveHeight=1,NegativeHeight=1)
    FalloffSize=0.4
    FlipHeightScaling=False
    FlipWidthScaling=False
    OptionalParentItem=(Type=Bone,Name="clavicle_r")
    RotationOffset=(X=177.7529,Y=58.9399,Z=-17.2984)
   )
   > &RigUnit_ModifyTransforms(
    ItemToModify="((Item=(Type=Bone,Name=\"dyn_hair_front_r_base\"),Transform=(Rotation=(X=0.3264,Y=-0.0058,Z=-0.1298,W=-0.9363),Translation=(X=-1.1383,Y=2.7771,Z=-3.0242),Scale3D=(X=1,Y=1,Z=1))),(Item=(Type=Bone,Name=\"dyn_hair_side_r_base\"),Transform=(Rotation=(X=-0.3192,Y=- ...(truncated)"
    WeightMinimum=0
    Weight=@n4.OutputParam
   )
   > &RigUnit_ModifyTransforms(
    ItemToModify="((Item=(Type=Bone,Name=\"dyn_hair_front_r_base\"),Transform=(Rotation=(X=0,Y=0,Z=0,W=1),Translation=(X=0,Y=0,Z=0),Scale3D=(X=1,Y=1,Z=1))),(Item=(Type=Bone,Name=\"dyn_hair_side_r_base\"),Transform=(Rotation=(X=0,Y=0,Z=0,W=1),Translation=(X=0,Y=0,Z=0),Scale3D=(X ...(truncated)"
    WeightMinimum=0
    Weight=@n5.OutputParam
   )
   n6 = &RigUnit_SphericalPoseReader(
    ActiveRegionScaleFactors=(PositiveWidth=1,NegativeWidth=3,PositiveHeight=1.5,NegativeHeight=3)
    ActiveRegionSize=0.1
    Debug=(bDrawDebug=false,bDraw2D=False,bDrawLocalAxes=False,DebugScale=25,DebugSegments=20,DebugThickness=0.25)
    DriverAxis=(X=1,Y=0,Z=0)
    DriverItem=(Type=Bone,Name="upperarm_l")
    FalloffRegionScaleFactors=(PositiveWidth=1,NegativeWidth=1,PositiveHeight=1,NegativeHeight=1)
    FalloffSize=0.4
    FlipHeightScaling=False
    FlipWidthScaling=False
    OptionalParentItem=(Type=Bone,Name="clavicle_l")
    RotationOffset=(X=33.1641,Y=-28.2741,Z=-6.8683)
    ExecutePin=@RigVMFunction_Sequence.B
   )
   > n7 = &RigUnit_SphericalPoseReader(
    ActiveRegionScaleFactors=(PositiveWidth=1,NegativeWidth=3,PositiveHeight=1.5,NegativeHeight=3)
    ActiveRegionSize=0.1
    Debug=(bDrawDebug=false,bDraw2D=False,bDrawLocalAxes=False,DebugScale=25,DebugSegments=20,DebugThickness=0.25)
    DriverAxis=(X=1,Y=0,Z=0)
    DriverItem=(Type=Bone,Name="upperarm_l")
    FalloffRegionScaleFactors=(PositiveWidth=1,NegativeWidth=1,PositiveHeight=1,NegativeHeight=1)
    FalloffSize=0.4
    FlipHeightScaling=False
    FlipWidthScaling=False
    OptionalParentItem=(Type=Bone,Name="clavicle_l")
    RotationOffset=(X=-4.3201,Y=-25.3565,Z=-3.6348)
   )
   > &RigUnit_ModifyTransforms(
    ItemToModify="((Item=(Type=Bone,Name=\"dyn_hair_front_l_base\"),Transform=(Rotation=(X=0.0153,Y=0.0063,Z=-0.1092,W=-0.9939),Translation=(X=1.3695,Y=-2.0655,Z=0.0816),Scale3D=(X=1,Y=1,Z=1))),(Item=(Type=Bone,Name=\"dyn_hair_side_l_base\"),Transform=(Rotation=(X=-0.145,Y=-0.0 ...(truncated)"
    WeightMinimum=0
    Weight=@n6.OutputParam
   )
   > &RigUnit_ModifyTransforms(
    ItemToModify="((Item=(Type=Bone,Name=\"dyn_hair_front_l_base\"),Transform=(Rotation=(X=0,Y=0,Z=0,W=1),Translation=(X=0,Y=0,Z=0),Scale3D=(X=1,Y=1,Z=1))),(Item=(Type=Bone,Name=\"dyn_hair_side_l_base\"),Transform=(Rotation=(X=0,Y=0,Z=0,W=1),Translation=(X=0,Y=0,Z=0),Scale3D=(X ...(truncated)"
    WeightMinimum=0
    Weight=@n7.OutputParam
   )
  ]}
]
hierarchy [
 root {loc=<0 0 0>, rot=<0 0 0 1>}
 pelvis {parent=root, loc=<0 -0.5301 98.0571>, rot=<0.0127 0.707 -0.0127 -0.707>}
 spine_01 {parent=pelvis, loc=<6.1208 -0.9328 0>, rot=<0 0 0.034 -0.9994>}
 spine_02 {parent=spine_01, loc=<6.5997 -0.2095 0>, rot=<0 0 0.0238 -0.9997>}
 spine_03 {parent=spine_02, loc=<6.4914 -0.1237 0>, rot=<0 0 0.0149 -0.9999>}
 spine_04 {parent=spine_03, loc=<6.9598 0.0794 0>, rot=<0 0 -0.0132 -0.9999>}
 spine_05 {parent=spine_04, loc=<6.9987 0 0>, rot=<0 0 -0.0015 -1>}
 neck_01 {parent=spine_05, loc=<11.3138 0.1529 0>, rot=<0 0 -0.2003 -0.9797>}
 neck_02 {parent=neck_01, loc=<4.5627 0 0>, rot=<0 0 0.0504 -0.9987>}
 head {parent=neck_02, loc=<5.3914 0 0>, rot=<0 0 0.115 -0.9934>}
 dyn_hair_base {parent=head, loc=<0.5 0 0>, rot=<0 0 0 1>}
 dyn_hair_rear_l_base {parent=dyn_hair_base, loc=<7.603 -5.3457 -5.1836>, rot=<0.1207 -0.9819 -0.1455 -0.0058>}
 dyn_hair_rear_l_01 {parent=dyn_hair_rear_l_base, loc=<0.2 0 0>, rot=<0 0 0 1>}
 dyn_hair_rear_l_02 {parent=dyn_hair_rear_l_01, loc=<7.7327 0.0231 0>, rot=<0 0 0.0083 -1>}
 dyn_hair_rear_l_03 {parent=dyn_hair_rear_l_02, loc=<5.9484 0 0>, rot=<0 0 0.0431 -0.9991>}
 dyn_hair_rear_l_04 {parent=dyn_hair_rear_l_03, loc=<5.5 0 0>, rot=<0 0 -0.0876 -0.9962>}
 dyn_hair_rear_base {parent=dyn_hair_base, loc=<7.5379 -7.6463 -0.4342>, rot=<0.1123 -0.9937 0 0>}
 dyn_hair_rear_01 {parent=dyn_hair_rear_base, loc=<0.2 0 0>, rot=<0 0 0 1>}
 dyn_hair_rear_02 {parent=dyn_hair_rear_01, loc=<8.1738 0 0>, rot=<0 0 0.0652 -0.9979>}
 dyn_hair_rear_03 {parent=dyn_hair_rear_02, loc=<5.5 0 0>, rot=<0 0 -0.1141 -0.9935>}
 dyn_hair_rear_04 {parent=dyn_hair_rear_03, loc=<5.5 0 0>, rot=<0 0 -0.0471 -0.9989>}
 dyn_hair_rear_05 {parent=dyn_hair_rear_04, loc=<5.6065 0 0>, rot=<0 0 0 1>}
 dyn_hair_front_l_base {parent=dyn_hair_base, loc=<7.4161 3.9418 -6.378>, rot=<-0.0986 0.6732 0.7328 0.0065>}
 dyn_hair_front_l_01 {parent=dyn_hair_front_l_base, loc=<0.2 0 0>, rot=<0 0 0 1>}
 dyn_hair_front_l_02 {parent=dyn_hair_front_l_01, loc=<8.3617 0 0>, rot=<0.0317 0.0613 -0.0995 -0.9926>}
 dyn_hair_front_l_03 {parent=dyn_hair_front_l_02, loc=<6.4724 0 0>, rot=<0.0004 0.1938 -0.002 -0.981>}
 dyn_hair_front_l_04 {parent=dyn_hair_front_l_03, loc=<6.2499 0 0>, rot=<0 0.0521 0 -0.9986>}
 dyn_hair_rear_r_base {parent=dyn_hair_base, loc=<7.2149 -3.8506 4.9644>, rot=<0.9397 0.1571 -0.009 0.3038>}
 dyn_hair_rear_r_01 {parent=dyn_hair_rear_r_base, loc=<-0.1995 0.0002 0>, rot=<0 0 0 1>}
 dyn_hair_rear_r_02 {parent=dyn_hair_rear_r_01, loc=<-7.7327 -0.0231 0>, rot=<0 0 0.0083 -1>}
 dyn_hair_rear_r_03 {parent=dyn_hair_rear_r_02, loc=<-5.9484 0 0>, rot=<0 0 0.0431 -0.9991>}
 dyn_hair_rear_r_04 {parent=dyn_hair_rear_r_03, loc=<-5.5001 0 0>, rot=<0 0 -0.0876 -0.9962>}
 dyn_hair_front_r_base {parent=dyn_hair_base, loc=<7.4165 3.9418 6.378>, rot=<-0.6732 -0.0986 0.0065 -0.7328>}
 dyn_hair_front_r_01 {parent=dyn_hair_front_r_base, loc=<-0.2007 -0.0001 -0.0001>, rot=<0 0 0 1>}
 dyn_hair_front_r_02 {parent=dyn_hair_front_r_01, loc=<-8.361 0.0001 0.0001>, rot=<0.0317 0.0613 -0.0995 -0.9926>}
 dyn_hair_front_r_03 {parent=dyn_hair_front_r_02, loc=<-6.4727 0 0>, rot=<0.0004 0.1938 -0.002 -0.981>}
 dyn_hair_front_r_04 {parent=dyn_hair_front_r_03, loc=<-6.2496 0 -0.0001>, rot=<0 0.0521 0 -0.9986>}
 dyn_hair_front_upper_l_base {parent=dyn_hair_base, loc=<9.189 6.4115 -3.6207>, rot=<-0.1687 0.666 0.7139 0.1356>}
 dyn_hair_front_upper_l_01 {parent=dyn_hair_front_upper_l_base, loc=<0.2 0 0>, rot=<0 0 0 1>}
 dyn_hair_front_upper_l_02 {parent=dyn_hair_front_upper_l_01, loc=<5.6053 0 0>, rot=<0.0259 0.0077 -0.0989 -0.9947>}
 dyn_hair_front_upper_r_base {parent=dyn_hair_base, loc=<9.7827 6.4703 4.0289>, rot=<-0.6698 -0.1531 0.1189 -0.7168>}
 dyn_hair_front_upper_r_01 {parent=dyn_hair_front_upper_r_base, loc=<-0.1997 0.0001 0>, rot=<0 0 0 1>}
 dyn_hair_front_upper_r_02 {parent=dyn_hair_front_upper_r_01, loc=<-5.6059 -0.0003 0>, rot=<0.0259 0.0077 -0.0989 -0.9947>}
 dyn_hair_side_l_base {parent=dyn_hair_base, loc=<8.0141 -1.392 -5.7869>, rot=<0.1464 -0.7712 -0.6187 -0.0313>}
 dyn_hair_side_l_01 {parent=dyn_hair_side_l_base, loc=<0.2 0 0>, rot=<0 0 0 1>}
 dyn_hair_side_l_02 {parent=dyn_hair_side_l_01, loc=<8.6989 0 0>, rot=<0.0196 0.2071 -0.0318 -0.9776>}
 dyn_hair_side_l_03 {parent=dyn_hair_side_l_02, loc=<6.4724 0 0>, rot=<0.0763 0.1965 -0.1118 -0.9711>}
 dyn_hair_side_l_04 {parent=dyn_hair_side_l_03, loc=<6.2499 0 0>, rot=<0 0.0521 0 -0.9986>}
 dyn_hair_side_r_base {parent=dyn_hair_base, loc=<8.0146 -1.392 5.7869>, rot=<-0.7712 -0.1464 0.0313 -0.6187>}
 dyn_hair_side_r_01 {parent=dyn_hair_side_r_base, loc=<-0.2006 -0.0002 -0.0001>, rot=<0 0 0 -1>}
 dyn_hair_side_r_02 {parent=dyn_hair_side_r_01, loc=<-8.6985 0.0001 0>, rot=<0.0196 0.2071 -0.0318 -0.9776>}
 dyn_hair_side_r_03 {parent=dyn_hair_side_r_02, loc=<-6.4729 -0.0001 0.0001>, rot=<0.0763 0.1965 -0.1118 -0.9711>}
 dyn_hair_side_r_04 {parent=dyn_hair_side_r_03, loc=<-6.2495 0 -0.0003>, rot=<0 0.0521 0 -0.9986>}
 clavicle_l {parent=spine_05, loc=<5.4831 -0.1053 -3.2946>, rot=<0.0465 -0.7289 -0.009 -0.6829>}
 upperarm_l {parent=clavicle_l, loc=<9.8658 -1.2136 -0.4105>, rot=<0.0164 -0.3245 -0.0269 -0.9454>}
 upperarm_twist_01_l {parent=upperarm_l, loc=<7.2135 -0.1134 0>, rot=<0 0 0 1>}
 deform_upperarm_bk_base_l {parent=upperarm_twist_01_l, loc=<-6.4053 -4.3746 0.314>, rot=<0 0 0 1>}
 deform_upperarm_bk_l {parent=deform_upperarm_bk_base_l, loc=<0 0 0>, rot=<0 0 0 1>}
 deform_upperarm_fr_base_l {parent=upperarm_twist_01_l, loc=<-5.6312 5.092 0.3354>, rot=<0 0 0 1>}
 deform_upperarm_fr_l {parent=deform_upperarm_fr_base_l, loc=<0 0 0>, rot=<0 0 0 1>}
 deform_upperarm_up_base_l {parent=upperarm_twist_01_l, loc=<-8.891 0.4825 3.9517>, rot=<0 0 0 1>}
 deform_upperarm_up_l {parent=deform_upperarm_up_base_l, loc=<0 0 0>, rot=<0 0 0 1>}
 deform_hair_split_l {parent=clavicle_l, loc=<5.4677 -2.386 7.7139>, rot=<0 0 0 1>}
 clavicle_r {parent=spine_05, loc=<5.4831 -0.1053 3.2946>, rot=<0.7289 0.0465 -0.6829 0.009>}
 upperarm_r {parent=clavicle_r, loc=<-9.8658 1.2136 0.4105>, rot=<0.0164 -0.3245 -0.0269 -0.9454>}
 upperarm_twist_01_r {parent=upperarm_r, loc=<-7.2135 0.1134 0>, rot=<0 0 0 -1>}
 deform_upperarm_bk_base_r {parent=upperarm_twist_01_r, loc=<6.4053 4.3746 -0.314>, rot=<0 0 0 1>}
 deform_upperarm_bk_r {parent=deform_upperarm_bk_base_r, loc=<0 0 0>, rot=<0 0 0 1>}
 deform_upperarm_fr_base_r {parent=upperarm_twist_01_r, loc=<5.6312 -5.092 -0.3354>, rot=<0 0 0 1>}
 deform_upperarm_fr_r {parent=deform_upperarm_fr_base_r, loc=<0 0 0>, rot=<0 0 0 1>}
 deform_upperarm_up_base_r {parent=upperarm_twist_01_r, loc=<8.891 -0.4825 -3.9517>, rot=<0 0 0 1>}
 deform_upperarm_up_r {parent=deform_upperarm_up_base_r, loc=<0 0 0>, rot=<0 0 0 1>}
 deform_hair_split_r {parent=clavicle_r, loc=<-5.4676 2.386 -7.7142>, rot=<0 0 0 1>}
 attach_backpack {parent=spine_05, loc=<-0.123 -13.473 0>, rot=<-0.0093 -0.707 0.0093 -0.707>}
 attach_cape {parent=spine_05, loc=<9.017 -11.714 0>, rot=<0.0616 -0.7044 -0.0616 -0.7044>}
 deform_pec_aim_base_l {parent=spine_04, loc=<3.2456 9.3309 -1.2237>, rot=<0.0603 -0.4033 0.3318 -0.8507>}
 deform_pec_aim_l {parent=deform_pec_aim_base_l, loc=<0 0 0>, rot=<0 0 0 1>}
 deform_pec_base_l {parent=deform_pec_aim_l, loc=<9.3665 0 0>, rot=<0 0 0 1>}
 deform_pec_l {parent=deform_pec_base_l, loc=<0 0 0>, rot=<0 0 0 1>}
 deform_pec_aim_base_r {parent=spine_04, loc=<3.2452 9.3309 1.2237>, rot=<-0.4033 -0.0603 0.8507 0.3318>}
 deform_pec_aim_r {parent=deform_pec_aim_base_r, loc=<0 0 0>, rot=<0 0 0 1>}
 deform_pec_base_r {parent=deform_pec_aim_r, loc=<-9.3667 -0.0003 -0.0004>, rot=<0 0 0 1>}
 deform_pec_r {parent=deform_pec_base_r, loc=<0 0 0>, rot=<0 0 0 1>}
]
variables [
 Is Wearing Backbling {type=bool, default=False}
 Is Wearing Cape {type=bool, default=False}
]
