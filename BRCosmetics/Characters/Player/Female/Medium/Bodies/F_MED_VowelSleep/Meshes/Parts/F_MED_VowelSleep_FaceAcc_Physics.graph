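# Physics asset for the VowelSleep female medium character's face accessory part, defining capsule bodies for dynamic hair and the upper body around it.
# Seven hair chains (front, side and rear) use angular limits that mostly loosen toward the tips, with bases linked angularly free to neck_01.
# Free-angle cross-links tie the center rear chain's fourth bone to both side rear tips.
# Spine, clavicle, upper arm, neck and head capsules supply body shapes.
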PhysicsAsset /BRCosmetics/Characters/Player/Female/Medium/Bodies/F_MED_VowelSleep/Meshes/Parts/F_MED_VowelSleep_FaceAcc_Physics
bodies {
 clavicle_l [{shapeType=Capsule, length=7.3614, radius=4.7877, shapeName=clavicle_l_capsule, center=<6.8896 -0.5 -0.3359>, rotation=<rot 79.0876 0 0>}]
 clavicle_r [{shapeType=Capsule, length=7.3614, radius=4.7877, shapeName=clavicle_r_capsule, center=<-6.8896 0.5 0.3359>, rotation=<rot 79.0876 0 180>}]
 dyn_hair_front_l_01 [{
   shapeType=Capsule
   length=3.5531
   radius=4
   shapeName=dyn_hair_front_l_01_capsule
   center=<2 -4.222 -0.0326>
   rotation=<rot 0.0036 0.0002 6.598>
  }]
 dyn_hair_front_l_02 [{shapeType=Capsule, length=5.7904, radius=2.6494, shapeName=dyn_hair_front_l_02_capsule, center=<2 -1.5196 0>, rotation=<rot -69.9999 0 0>}]
 dyn_hair_front_l_03 [{
   shapeType=Capsule
   length=3.8252
   radius=3.1263
   shapeName=dyn_hair_front_l_03_capsule
   center=<2.488 -0.1292 0.5294>
   rotation=<rot -9.6217 -94.5852 82.5683>
  }]
 dyn_hair_front_l_04 [
  {
   shapeType=Capsule
   length=4.48
   radius=3.7789
   shapeName=dyn_hair_front_l_04_capsule
   center=<4.2986 -0.5207 3.5518>
   rotation=<rot -5.8959 -79.9463 64.1523>
  }
  {
   shapeType=Capsule
   length=3.8998
   radius=3.5347
   shapeName=dyn_hair_front_l_04_capsule
   center=<3.8579 -4.3477 1.145>
   rotation=<rot -70.1783 0 -18.1688>
  }
 ]
 dyn_hair_front_l_base [{shapeType=Capsule, length=0.505, radius=0.505, shapeName=dyn_hair_front_l_base_capsule, rotation=<rot 0 -90 0>}]
 dyn_hair_front_r_01 [{
   shapeType=Capsule
   length=6
   radius=4
   shapeName=dyn_hair_front_r_01_capsule
   center=<-2 4.222 0.0326>
   rotation=<rot 0.0036 0.0002 -173.402>
  }]
 dyn_hair_front_r_02 [{shapeType=Capsule, length=5.7904, radius=2.6494, shapeName=dyn_hair_front_r_02_capsule, center=<-2 1.5196 0>, rotation=<rot -69.9999 0 -180>}]
 dyn_hair_front_r_03 [{
   shapeType=Capsule
   length=3.8252
   radius=3.1263
   shapeName=dyn_hair_front_r_03_capsule
   center=<-2.488 0.1292 -0.5294>
   rotation=<rot -9.6217 -94.5852 -97.4317>
  }]
 dyn_hair_front_r_04 [{
   shapeType=Capsule
   length=2.7195
   radius=3.3963
   shapeName=dyn_hair_front_r_04_capsule
   center=<-4.6194 0.4851 -3.755>
   rotation=<rot -5.8947 -79.9463 -106.0908>
  }]
 dyn_hair_front_r_base [{shapeType=Capsule, length=0.505, radius=0.505, shapeName=dyn_hair_front_r_base_capsule, rotation=<rot 0 -90 180>}]
 dyn_hair_rear_01 [{shapeType=Capsule, length=8, radius=4, shapeName=dyn_hair_rear_01_capsule, center=<2 -3 0>}]
 dyn_hair_rear_02 [{shapeType=Capsule, length=8, radius=4, shapeName=dyn_hair_rear_02_capsule, center=<2 -1.2 0>}]
 dyn_hair_rear_03 [{shapeType=Capsule, length=8, radius=4, shapeName=dyn_hair_rear_03_capsule, center=<2 0 0>}]
 dyn_hair_rear_04 [{shapeType=Capsule, length=8, radius=4, shapeName=dyn_hair_rear_04_capsule, center=<2 0 0>}]
 dyn_hair_rear_05 [{shapeType=Capsule, length=8, radius=4, shapeName=dyn_hair_rear_05_capsule, center=<3.3649 -1.7097 0>}]
 dyn_hair_rear_base [{shapeType=Capsule, length=0.505, radius=0.505, shapeName=dyn_hair_rear_base_capsule, rotation=<rot 0 -90 0>}]
 dyn_hair_rear_l_01 [{shapeType=Capsule, length=8, radius=4, shapeName=dyn_hair_rear_l_01_capsule, center=<2 -4.3917 0>}]
 dyn_hair_rear_l_02 [{shapeType=Capsule, length=8, radius=4, shapeName=dyn_hair_rear_l_02_capsule, center=<2 -2.2 0>}]
 dyn_hair_rear_l_03 [{shapeType=Capsule, length=8, radius=4, shapeName=dyn_hair_rear_l_03_capsule, center=<2 0 0>}]
 dyn_hair_rear_l_04 [{
   shapeType=Capsule
   length=4.9174
   radius=4
   shapeName=dyn_hair_rear_l_04_capsule
   center=<1.9996 -1.1395 0.2069>
   rotation=<rot 0.0206 0.0688 -15.6105>
  }]
 dyn_hair_rear_l_base [{shapeType=Capsule, length=0.505, radius=0.505, shapeName=dyn_hair_rear_l_base_capsule, rotation=<rot 0 -90 0>}]
 dyn_hair_rear_r_01 [{shapeType=Capsule, length=8, radius=4, shapeName=dyn_hair_rear_r_01_capsule, center=<-2 4.3917 0>, rotation=<rot 0 0 -180>}]
 dyn_hair_rear_r_02 [{shapeType=Capsule, length=8, radius=4, shapeName=dyn_hair_rear_r_02_capsule, center=<-2 2.2 0>, rotation=<rot 0 0 -180>}]
 dyn_hair_rear_r_03 [{shapeType=Capsule, length=8, radius=4, shapeName=dyn_hair_rear_r_03_capsule, center=<-2 0 0>, rotation=<rot 0 0 -180>}]
 dyn_hair_rear_r_04 [{
   shapeType=Capsule
   length=4.1514
   radius=4
   shapeName=dyn_hair_rear_r_04_capsule
   center=<-2.0027 0.7682 1.0911>
   rotation=<rot 1.0478 3.1817 147.8338>
  }]
 dyn_hair_rear_r_base [{shapeType=Capsule, length=0.505, radius=0.505, shapeName=dyn_hair_rear_r_base_capsule, rotation=<rot 0 -90 0>}]
 dyn_hair_side_l_01 [{shapeType=Capsule, length=5, radius=3, shapeName=dyn_hair_side_l_01_capsule}]
 dyn_hair_side_l_02 [{shapeType=Capsule, length=5, radius=3, shapeName=dyn_hair_side_l_02_capsule, rotation=<rot -75.8941 -135.4381 134.561>}]
 dyn_hair_side_l_03 [
  {
   shapeType=Capsule
   length=5.9342
   radius=2.0658
   shapeName=dyn_hair_side_l_03_capsule
   center=<0 0.4925 0.8531>
   rotation=<rot -60 -89.9997 89.9996>
  }
  {
   shapeType=Capsule
   length=4.0499
   radius=2.0153
   shapeName=dyn_hair_side_l_03_capsule
   center=<7.5253 -2.2977 -0.7743>
   rotation=<rot -52.0018 43.8281 -43.8595>
  }
 ]
 dyn_hair_side_l_04 [{
   shapeType=Capsule
   length=2.8918
   radius=2.4252
   shapeName=dyn_hair_side_l_04_capsule
   center=<0 -0.154 3.0638>
   rotation=<rot -39.2734 -77.0374 81.7096>
  }]
 dyn_hair_side_l_base [{shapeType=Capsule, length=0.505, radius=0.505, shapeName=dyn_hair_side_l_base_capsule, rotation=<rot 0 -90 0>}]
 dyn_hair_side_r_01 [{shapeType=Capsule, length=5, radius=3, shapeName=dyn_hair_side_r_01_capsule}]
 dyn_hair_side_r_02 [{shapeType=Capsule, length=5, radius=3, shapeName=dyn_hair_side_r_02_capsule, rotation=<rot -75.8941 -135.4381 -45.439>}]
 dyn_hair_side_r_03 [{
   shapeType=Capsule
   length=7.9364
   radius=2.8146
   shapeName=dyn_hair_side_r_03_capsule
   center=<-1.665 -0.2427 -0.4204>
   rotation=<rot -55.921 -58.5853 -116.8334>
  }]
 dyn_hair_side_r_04 [{
   shapeType=Capsule
   length=2.8918
   radius=2.4252
   shapeName=dyn_hair_side_r_04_capsule
   center=<0 0.15 -3>
   rotation=<rot -39.2734 -77.0374 -98.2904>
  }]
 dyn_hair_side_r_base [{shapeType=Capsule, length=0.505, radius=0.505, shapeName=dyn_hair_side_r_base_capsule, rotation=<rot 0 -90 0>}]
 head [{shapeType=Capsule, length=9.3383, radius=5.8115, shapeName=head_capsule, center=<1.2763 1.7133 0>, rotation=<rot 0 -102.4218 90>}]
 neck_01 [{shapeType=Capsule, length=6.9189, radius=5.1251, shapeName=neck_01_capsule, center=<0.2007 0.9415 0>, rotation=<rot 90 -12.0315 0>}]
 spine_04 [
  {shapeType=Capsule, length=11.1972, radius=8.843, shapeName=spine_04_capsule, center=<4.1915 0 0>}
  {shapeType=Capsule, length=8.1952, radius=8.025, shapeName=spine_04_capsule, center=<3.6137 4.1056 0>}
 ]
 spine_05 [{shapeType=Capsule, length=19.8828, radius=6.4005, shapeName=spine_05_capsule, center=<3.1164 -0.4742 0>, rotation=<rot 0 -90 0>}]
 upperarm_l [{shapeType=Capsule, length=26.1842, radius=4.0658, shapeName=upperarm_l_capsule, center=<13 -0.0532 -1.1454>, rotation=<rot 90 -3.8402 0>}]
 upperarm_r [{shapeType=Capsule, length=26.1842, radius=4.0658, shapeName=upperarm_r_capsule, center=<-13 0.0532 1.1454>, rotation=<rot 90 176.1598 0>}]
}
constraints [
 {bone1Name=dyn_hair_rear_01, bone2Name=dyn_hair_rear_base, swing1LimitDegrees=14, swing2LimitDegrees=8, twistLimitDegrees=8}
 {bone1Name=dyn_hair_rear_02, bone2Name=dyn_hair_rear_01, swing1LimitDegrees=14, swing2LimitDegrees=8, twistLimitDegrees=8}
 {bone1Name=dyn_hair_rear_03, bone2Name=dyn_hair_rear_02, swing1LimitDegrees=32, swing2LimitDegrees=12, twistLimitDegrees=8}
 {bone1Name=dyn_hair_rear_04, bone2Name=dyn_hair_rear_03, swing1LimitDegrees=24, swing2LimitDegrees=12, twistLimitDegrees=8}
 {bone1Name=dyn_hair_rear_05, bone2Name=dyn_hair_rear_04, swing1LimitDegrees=24, swing2LimitDegrees=16, twistLimitDegrees=12}
 {bone1Name=dyn_hair_rear_l_01, bone2Name=dyn_hair_rear_l_base, swing1LimitDegrees=14, swing2LimitDegrees=4, twistLimitDegrees=4}
 {bone1Name=dyn_hair_rear_l_02, bone2Name=dyn_hair_rear_l_01, swing1LimitDegrees=14, swing2LimitDegrees=4, twistLimitDegrees=4}
 {bone1Name=dyn_hair_rear_l_03, bone2Name=dyn_hair_rear_l_02, swing1LimitDegrees=24, swing2LimitDegrees=12, twistLimitDegrees=8}
 {bone1Name=dyn_hair_rear_l_04, bone2Name=dyn_hair_rear_l_03, swing1LimitDegrees=24, swing2LimitDegrees=12, twistLimitDegrees=8}
 {bone1Name=dyn_hair_rear_r_01, bone2Name=dyn_hair_rear_r_base, swing1LimitDegrees=14, swing2LimitDegrees=4, twistLimitDegrees=4}
 {bone1Name=dyn_hair_rear_r_02, bone2Name=dyn_hair_rear_r_01, swing1LimitDegrees=14, swing2LimitDegrees=4, twistLimitDegrees=4}
 {bone1Name=dyn_hair_rear_r_03, bone2Name=dyn_hair_rear_r_02, swing1LimitDegrees=24, swing2LimitDegrees=12, twistLimitDegrees=8}
 {bone1Name=dyn_hair_rear_r_04, bone2Name=dyn_hair_rear_r_03, swing1LimitDegrees=24, swing2LimitDegrees=12, twistLimitDegrees=8}
 {
  bone1Name=dyn_hair_rear_04
  bone2Name=dyn_hair_rear_l_04
  swing1LimitDegrees=45
  swing1Motion=Free
  swing2LimitDegrees=45
  swing2Motion=Free
  twistLimitDegrees=45
  twistMotion=Free
 }
 {
  bone1Name=dyn_hair_rear_04
  bone2Name=dyn_hair_rear_r_04
  swing1LimitDegrees=45
  swing1Motion=Free
  swing2LimitDegrees=45
  swing2Motion=Free
  twistLimitDegrees=45
  twistMotion=Free
 }
 {bone1Name=dyn_hair_front_l_01, bone2Name=dyn_hair_front_l_base, swing1LimitDegrees=8, swing2LimitDegrees=8, twistLimitDegrees=8}
 {bone1Name=dyn_hair_front_l_02, bone2Name=dyn_hair_front_l_01, swing1LimitDegrees=8, swing2LimitDegrees=8, twistLimitDegrees=8}
 {bone1Name=dyn_hair_front_l_03, bone2Name=dyn_hair_front_l_02, swing1LimitDegrees=20, swing2LimitDegrees=20, twistLimitDegrees=20}
 {bone1Name=dyn_hair_front_l_04, bone2Name=dyn_hair_front_l_03, swing1LimitDegrees=24, swing2LimitDegrees=24, twistLimitDegrees=24}
 {bone1Name=dyn_hair_front_r_01, bone2Name=dyn_hair_front_r_base, swing1LimitDegrees=8, swing2LimitDegrees=8, twistLimitDegrees=8}
 {bone1Name=dyn_hair_front_r_02, bone2Name=dyn_hair_front_r_01, swing1LimitDegrees=8, swing2LimitDegrees=8, twistLimitDegrees=8}
 {bone1Name=dyn_hair_front_r_03, bone2Name=dyn_hair_front_r_02, swing1LimitDegrees=20, swing2LimitDegrees=20, twistLimitDegrees=20}
 {bone1Name=dyn_hair_front_r_04, bone2Name=dyn_hair_front_r_03, swing1LimitDegrees=24, swing2LimitDegrees=24, twistLimitDegrees=24}
 {bone1Name=dyn_hair_side_l_01, bone2Name=dyn_hair_side_l_base, swing1LimitDegrees=8, swing2LimitDegrees=8, twistLimitDegrees=8}
 {bone1Name=dyn_hair_side_l_02, bone2Name=dyn_hair_side_l_01, swing1LimitDegrees=8, swing2LimitDegrees=8, twistLimitDegrees=4}
 {bone1Name=dyn_hair_side_l_03, bone2Name=dyn_hair_side_l_02, swing1LimitDegrees=20, swing2LimitDegrees=20, twistLimitDegrees=4}
 {bone1Name=dyn_hair_side_l_04, bone2Name=dyn_hair_side_l_03, swing1LimitDegrees=24, swing2LimitDegrees=24, twistLimitDegrees=4}
 {bone1Name=dyn_hair_side_r_01, bone2Name=dyn_hair_side_r_base, swing1LimitDegrees=8, swing2LimitDegrees=8, twistLimitDegrees=8}
 {bone1Name=dyn_hair_side_r_02, bone2Name=dyn_hair_side_r_01, swing1LimitDegrees=8, swing2LimitDegrees=8, twistLimitDegrees=4}
 {bone1Name=dyn_hair_side_r_03, bone2Name=dyn_hair_side_r_02, swing1LimitDegrees=18, swing2LimitDegrees=18, twistLimitDegrees=4}
 {bone1Name=dyn_hair_side_r_04, bone2Name=dyn_hair_side_r_03, swing1LimitDegrees=24, swing2LimitDegrees=24, twistLimitDegrees=4}
 {
  bone1Name=head
  bone2Name=neck_01
  swing1LimitDegrees=45
  swing1Motion=Free
  swing2LimitDegrees=45
  swing2Motion=Free
  twistLimitDegrees=45
  twistMotion=Free
 }
 {
  bone1Name=dyn_hair_rear_l_base
  bone2Name=neck_01
  swing1LimitDegrees=45
  swing1Motion=Free
  swing2LimitDegrees=45
  swing2Motion=Free
  twistLimitDegrees=45
  twistMotion=Free
 }
 {
  bone1Name=dyn_hair_rear_base
  bone2Name=neck_01
  swing1LimitDegrees=45
  swing1Motion=Free
  swing2LimitDegrees=45
  swing2Motion=Free
  twistLimitDegrees=45
  twistMotion=Free
 }
 {
  bone1Name=dyn_hair_front_l_base
  bone2Name=neck_01
  swing1LimitDegrees=45
  swing1Motion=Free
  swing2LimitDegrees=45
  swing2Motion=Free
  twistLimitDegrees=45
  twistMotion=Free
 }
 {
  bone1Name=dyn_hair_rear_r_base
  bone2Name=neck_01
  swing1LimitDegrees=45
  swing1Motion=Free
  swing2LimitDegrees=45
  swing2Motion=Free
  twistLimitDegrees=45
  twistMotion=Free
 }
 {
  bone1Name=dyn_hair_front_r_base
  bone2Name=neck_01
  swing1LimitDegrees=45
  swing1Motion=Free
  swing2LimitDegrees=45
  swing2Motion=Free
  twistLimitDegrees=45
  twistMotion=Free
 }
 {
  bone1Name=dyn_hair_side_l_base
  bone2Name=neck_01
  swing1LimitDegrees=45
  swing1Motion=Free
  swing2LimitDegrees=45
  swing2Motion=Free
  twistLimitDegrees=45
  twistMotion=Free
 }
 {
  bone1Name=dyn_hair_side_r_base
  bone2Name=neck_01
  swing1LimitDegrees=45
  swing1Motion=Free
  swing2LimitDegrees=45
  swing2Motion=Free
  twistLimitDegrees=45
  twistMotion=Free
 }
 {
  bone1Name=neck_01
  bone2Name=spine_05
  swing1LimitDegrees=45
  swing1Motion=Free
  swing2LimitDegrees=45
  swing2Motion=Free
  twistLimitDegrees=45
  twistMotion=Free
 }
]
