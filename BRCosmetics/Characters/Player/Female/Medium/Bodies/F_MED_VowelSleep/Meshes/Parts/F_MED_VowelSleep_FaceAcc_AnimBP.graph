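# Post-process style AnimBlueprint for the VowelSleep female character's FaceAcc part, simulating its dynamic hair around worn back bling.
# The attached parent's pose is copied through a Control Rig, then feeds three cached spine_04-space RigidBody sims: normal, backpack (own physics asset plus rear-hair AnimDynamics) and fitted cape.
# Backbling checks and the VowelSleep meta tag select the cache, with additive bone offsets shaping the hair.
# The EventGraph also computes a gravity override from a parameter asset.
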AnimBlueprint /BRCosmetics/Characters/Player/Female/Medium/Bodies/F_MED_VowelSleep/Meshes/Parts/F_MED_VowelSleep_FaceAcc_AnimBP
target_skeleton=&/F_MED_VowelSleep_FaceAcc_Skeleton
graphs [
 AnimGraph {nodes [
   n1 = Blend Poses by bool(
    BlendPose_0=Use cached pose 'Fitted Cape Sim Cache'()
    BlendPose_1=Use cached pose 'Backpack Sim Cache'()
    bActiveValue=Get Fitted Cape() -> (Fitted Cape:bool)
   )
   n2 = Use cached pose 'Normal Sim Cache'()
   n3 = Get isWearingBackbling()
   n4 = "Transform (Modify) Bone - Bone: dyn_hair_side_r_02"!(
    ComponentPose=Local To Component(LocalPose=Blend Poses by bool(BlendPose_0=@n1, BlendPose_1=@n2, bActiveValue=@n3))
    bAlphaBoolEnabled:bool=True
    Translation="1,0,0"
    Rotation="3,0,7"
    Scale="1,1,1"
   ) -> (Pose) {
    alphaInputType=Bool
    rotationMode=BMM_Additive
    rotationSpace=BCS_BoneSpace
    translationMode=BMM_Additive
    translationSpace=BCS_BoneSpace
    boneToModify {boneName=dyn_hair_side_r_02}
    rotation {pitch=3, roll=7}
    translation {x=1}
   }
   n5 = "Transform (Modify) Bone - Bone: dyn_hair_side_r_03"!(ComponentPose=@n4, bAlphaBoolEnabled:bool=True, Translation="1,0,0", Rotation="5,0,5", Scale="1,1,1") -> (Pose) {
    alphaInputType=Bool
    rotationMode=BMM_Additive
    rotationSpace=BCS_BoneSpace
    translationMode=BMM_Additive
    translationSpace=BCS_BoneSpace
    boneToModify {boneName=dyn_hair_side_r_03}
    rotation {pitch=5, roll=5}
    translation {x=1}
   }
   n6 = "Transform (Modify) Bone - Bone: dyn_hair_side_r_03"(ComponentPose=@n5, bAlphaBoolEnabled=True, Translation="0,0,0", Rotation="3,0,0", Scale="1,1,1") {
    alphaInputType=Bool
    rotationMode=BMM_Additive
    rotationSpace=BCS_BoneSpace
    translationMode=BMM_Additive
    translationSpace=BCS_BoneSpace
    boneToModify {boneName=dyn_hair_side_r_03}
    rotation {pitch=3}
   }
   n7 = "Transform (Modify) Bone - Bone: dyn_hair_front_r_03"!(ComponentPose=@n6, bAlphaBoolEnabled:bool=True, Translation="0,0.7,0", Rotation="0,0,0", Scale="1,1,0.9") -> (Pose) {
    alphaInputType=Bool
    rotationMode=BMM_Additive
    rotationSpace=BCS_BoneSpace
    scaleMode=BMM_Additive
    scaleSpace=BCS_BoneSpace
    translationMode=BMM_Additive
    translationSpace=BCS_BoneSpace
    boneToModify {boneName=dyn_hair_front_r_03}
    scale {z=0.9}
    translation {y=0.7}
   }
   n8 = "Transform (Modify) Bone - Bone: dyn_hair_front_r_04"!(ComponentPose=@n7, bAlphaBoolEnabled:bool=True, Translation="0,0,-0.5", Rotation="0,0,3", Scale="1,1,1") -> (Pose) {
    alphaInputType=Bool
    rotationMode=BMM_Additive
    rotationSpace=BCS_BoneSpace
    scaleMode=BMM_Additive
    scaleSpace=BCS_BoneSpace
    translationMode=BMM_Additive
    translationSpace=BCS_BoneSpace
    boneToModify {boneName=dyn_hair_front_r_04}
    rotation {roll=3}
    translation {z=-0.5}
   }
   n9 = "Transform (Modify) Bone - Bone: dyn_hair_front_l_03"!(ComponentPose=@n8, bAlphaBoolEnabled:bool=True, Translation="0,0,0", Rotation="0,0,0", Scale="1,1,1") -> (Pose) {
    alphaInputType=Bool
    rotationMode=BMM_Additive
    rotationSpace=BCS_BoneSpace
    scaleMode=BMM_Additive
    scaleSpace=BCS_BoneSpace
    translationMode=BMM_Additive
    translationSpace=BCS_BoneSpace
    boneToModify {boneName=dyn_hair_front_l_03}
   }
   n10 = "Transform (Modify) Bone - Bone: dyn_hair_front_l_04"!(ComponentPose=@n9, bAlphaBoolEnabled:bool=True, Translation="0,0,0.7", Rotation="0,0,-3", Scale="1,0.9,1") -> (Pose) {
    alphaInputType=Bool
    rotationMode=BMM_Additive
    rotationSpace=BCS_BoneSpace
    scaleMode=BMM_Additive
    scaleSpace=BCS_BoneSpace
    translationMode=BMM_Additive
    translationSpace=BCS_BoneSpace
    boneToModify {boneName=dyn_hair_front_l_04}
    rotation {roll=-3}
    scale {y=0.9}
    translation {z=0.7}
   }
   Output Pose(Result=Component To Local(ComponentPose=@n10)) {layerGroup=None}
   n11 = Control Rig!(
    Source=Copy Pose From Mesh() {bUseAttachedParent=true}
    Alpha
    Is Wearing Backbling:bool=Get isWearingBackbling()
    Is Wearing Cape:bool=Get Is Backbling A Cape() -> (Is Backbling A Cape:bool)
   ) -> (Pose) {controlRigAssetReference {blueprintRigClass=&/F_MED_VowelSleep_FaceAcc_CtrlRig.F_MED_VowelSleep_FaceAcc_CtrlRig_C}}
   Save cached pose 'Post Deformation'(Pose=@n11) {cacheName=Post Deformation}
   n12 = RigidBody!(
    ComponentPose=Local To Component(LocalPose=Use cached pose 'Post Deformation'())
    Alpha
    OverrideWorldGravity:Vector="0,0,0"
    ExternalForce
    SimSpaceSettings:Sim Space Settings Structure=(WorldAlpha=0.9,VelocityScaleZ=0.5,DampingAlpha=0.2,MaxLinearVelocity=770,MaxAngularVelocity=2,MaxLinearAcceleration=2400,MaxAngularAcceleration=14,ExternalLinearDragV=(X=0.1,Y=0.1,Z=0.45),ExternalLinearVelocity=(X=0,Y=0,Z=0),ExternalAngularVelocity=(X=0,Y=0,Z=0))
   ) -> (Pose) {
    bOverrideWorldGravity=true
    overridePhysicsAsset=&/F_MED_VowelSleep_FaceAcc_Physics
    simulationSpace=BaseBoneSpace
    baseBoneRef {boneName=spine_04}
    componentAppliedLinearAccClamp=<0 0 0>
    simSpaceSettings {
     dampingAlpha=0.2
     maxAngularAcceleration=14
     maxAngularVelocity=2
     maxLinearAcceleration=2400
     maxLinearVelocity=770
     velocityScaleZ=0.5
     worldAlpha=0.9
     externalLinearDragV=<0.1 0.1 0.45>
    }
   }
   n13 = "Transform (Modify) Bone - Bone: dyn_hair_rear_l_03"(ComponentPose=@n12, Alpha=1, Translation="0,0.5,0", Rotation="0,0,0", Scale="1,1,1") {
    rotationMode=BMM_Additive
    rotationSpace=BCS_BoneSpace
    translationMode=BMM_Additive
    translationSpace=BCS_BoneSpace
    boneToModify {boneName=dyn_hair_rear_l_03}
    translation {y=0.5}
   }
   n14 = "Transform (Modify) Bone - Bone: dyn_hair_rear_l_04"(ComponentPose=@n13, Alpha=1, Translation="0,0.5,0", Rotation="0,5,0", Scale="1,1,1") {
    rotationMode=BMM_Additive
    rotationSpace=BCS_BoneSpace
    translationMode=BMM_Additive
    translationSpace=BCS_BoneSpace
    boneToModify {boneName=dyn_hair_rear_l_04}
    rotation {yaw=5}
    translation {y=0.5}
   }
   n15 = "Transform (Modify) Bone - Bone: dyn_hair_rear_03"(ComponentPose=@n14, Alpha=1, Translation="0,1,0", Rotation="0,-5,0", Scale="1,1,1") {
    rotationMode=BMM_Additive
    rotationSpace=BCS_BoneSpace
    translationMode=BMM_Additive
    translationSpace=BCS_BoneSpace
    boneToModify {boneName=dyn_hair_rear_03}
    rotation {yaw=-5}
    translation {y=1}
   }
   n16 = "Transform (Modify) Bone - Bone: dyn_hair_rear_04"(ComponentPose=@n15, Alpha=1, Translation="0,1,0", Rotation="0,5,0", Scale="1,1,1") {
    rotationMode=BMM_Additive
    rotationSpace=BCS_BoneSpace
    translationMode=BMM_Additive
    translationSpace=BCS_BoneSpace
    boneToModify {boneName=dyn_hair_rear_04}
    rotation {yaw=5}
    translation {y=1}
   }
   n17 = "Transform (Modify) Bone - Bone: dyn_hair_rear_r_03"(ComponentPose=@n16, Alpha=1, Translation="0,-1,0", Rotation="0,-5,0", Scale="1,1,1") {
    rotationMode=BMM_Additive
    rotationSpace=BCS_BoneSpace
    translationMode=BMM_Additive
    translationSpace=BCS_BoneSpace
    boneToModify {boneName=dyn_hair_rear_r_03}
    rotation {yaw=-5}
    translation {y=-1}
   }
   n18 = "Transform (Modify) Bone - Bone: dyn_hair_rear_r_04"(ComponentPose=@n17, Alpha=1, Translation="0,-1,0", Rotation="0,-5,0", Scale="1,1,1") {
    rotationMode=BMM_Additive
    rotationSpace=BCS_BoneSpace
    translationMode=BMM_Additive
    translationSpace=BCS_BoneSpace
    boneToModify {boneName=dyn_hair_rear_r_04}
    rotation {yaw=-5}
    translation {y=-1}
   }
   n19 = "Transform (Modify) Bone - Bone: dyn_hair_front_r_01"(ComponentPose=@n18, Alpha=1, Translation="0,0.2,0", Rotation="0,0,0", Scale="1,1,1") {
    rotationMode=BMM_Additive
    rotationSpace=BCS_BoneSpace
    scaleSpace=BCS_BoneSpace
    translationMode=BMM_Additive
    translationSpace=BCS_BoneSpace
    boneToModify {boneName=dyn_hair_front_r_01}
    translation {y=0.2}
   }
   n20 = "Transform (Modify) Bone - Bone: dyn_hair_front_l_03"!(ComponentPose=@n19, Alpha:float=1, Translation="0,0,0", Rotation="0,0,-10", Scale="1,1,1") -> (Pose) {
    rotationMode=BMM_Additive
    rotationSpace=BCS_BoneSpace
    scaleSpace=BCS_BoneSpace
    translationMode=BMM_Additive
    translationSpace=BCS_BoneSpace
    boneToModify {boneName=dyn_hair_front_l_03}
    rotation {roll=-10}
   }
   n21 = "Transform (Modify) Bone - Bone: dyn_hair_side_l_03"(ComponentPose=@n20, Alpha=1, Translation="0,0,0", Rotation="0,0,0", Scale="1,1,1") {
    rotationMode=BMM_Additive
    rotationSpace=BCS_BoneSpace
    scaleSpace=BCS_BoneSpace
    translationMode=BMM_Additive
    translationSpace=BCS_BoneSpace
    boneToModify {boneName=dyn_hair_side_l_03}
   }
   Save cached pose 'Normal Sim Cache'(Pose=Component To Local(ComponentPose=@n21)) {cacheName=Normal Sim Cache}
   n22 = RigidBody!(ComponentPose=Local To Component(LocalPose=Use cached pose 'Post Deformation'()), Alpha, OverrideWorldGravity="0,0,0", ExternalForce) -> (Pose) {
    bOverrideWorldGravity=true
    overridePhysicsAsset=&/F_MED_VowelSleep_FaceAcc_Backbling_Physics
    simulationSpace=BaseBoneSpace
    baseBoneRef {boneName=spine_04}
    componentAppliedLinearAccClamp=<0 0 0>
    simSpaceSettings {
     dampingAlpha=0.2
     maxAngularAcceleration=14
     maxAngularVelocity=2
     maxLinearAcceleration=2400
     maxLinearVelocity=770
     velocityScaleZ=0.4
     worldAlpha=0.8
     externalLinearDragV=<0.1 0.1 0.5>
    }
   }
   n23 = Anim Dynamics - Bone: dyn_hair_rear_l_03(ComponentPose=@n22, Alpha=1, ExternalForce="0,0,0") {
    angularDampingOverride=1
    angularSpringConstant=20000
    bAngularSpring=true
    bOverrideAngularDamping=true
    bUseGravityOverride=true
    gravityScale=0.35
    simulationSpace=RootRelative
    boundBone {boneName=dyn_hair_rear_l_03}
    componentAppliedLinearAccClamp=<0 0 0>
    gravityOverride {z=-980}
    physicsBodyDefinitions [{
      boundBone {boneName=dyn_hair_rear_l_03}
      boxExtents=<20 30 30>
      constraintSetup {angularLimitsMax=<2 10 2>, angularLimitsMin=<-2 -10 -2>, angularTarget=<1 0 0>}
      localJointOffset=<4 0 0>
     }]
    simSpaceSettings {maxAngularAcceleration=30, maxAngularVelocity=770, simSpaceAngularAlpha=1}
   }
   n24 = Anim Dynamics - Bone: dyn_hair_rear_l_04(ComponentPose=@n23, Alpha=1, ExternalForce="0,0,0") {
    angularDampingOverride=1
    angularSpringConstant=20000
    bAngularSpring=true
    bOverrideAngularDamping=true
    bUseGravityOverride=true
    gravityScale=0.35
    simulationSpace=RootRelative
    boundBone {boneName=dyn_hair_rear_l_04}
    componentAppliedLinearAccClamp=<0 0 0>
    gravityOverride {z=-980}
    physicsBodyDefinitions [{
      boundBone {boneName=dyn_hair_rear_l_04}
      boxExtents=<20 30 30>
      constraintSetup {angularLimitsMax=<2 10 2>, angularLimitsMin=<-2 -10 -2>, angularTarget=<1 0 0>}
      localJointOffset=<4 0 0>
     }]
    simSpaceSettings {maxAngularAcceleration=30, maxAngularVelocity=770, simSpaceAngularAlpha=1}
   }
   n25 = Anim Dynamics - Bone: dyn_hair_rear_03(
    ComponentPose="Transform (Modify) Bone - Bone: dyn_hair_rear_l_02"(ComponentPose=@n24, Alpha=1, Translation="0,0,0", Rotation="0,0,0", Scale="1,0.7,1") {scaleMode=BMM_Additive, scaleSpace=BCS_BoneSpace, boneToModify {boneName=dyn_hair_rear_l_02}, scale {y=0.7}}
    Alpha=1
    ExternalForce="0,0,0"
   ) {
    angularDampingOverride=1
    angularSpringConstant=20000
    bAngularSpring=true
    bOverrideAngularDamping=true
    bUseGravityOverride=true
    gravityScale=0.35
    simulationSpace=RootRelative
    boundBone {boneName=dyn_hair_rear_03}
    componentAppliedLinearAccClamp=<0 0 0>
    gravityOverride {z=-980}
    physicsBodyDefinitions [{
      boundBone {boneName=dyn_hair_rear_03}
      boxExtents=<20 30 30>
      constraintSetup {angularLimitsMax=<2 10 2>, angularLimitsMin=<-2 -10 -2>, angularTarget=<1 0 0>}
      localJointOffset=<4 0 0>
     }]
    simSpaceSettings {maxAngularAcceleration=30, maxAngularVelocity=770, simSpaceAngularAlpha=1}
   }
   n26 = Anim Dynamics - Bone: dyn_hair_rear_04(ComponentPose=@n25, Alpha=1, ExternalForce="0,0,0") {
    angularDampingOverride=1
    angularSpringConstant=20000
    bAngularSpring=true
    bOverrideAngularDamping=true
    bUseGravityOverride=true
    gravityScale=0.35
    simulationSpace=RootRelative
    boundBone {boneName=dyn_hair_rear_04}
    componentAppliedLinearAccClamp=<0 0 0>
    gravityOverride {z=-980}
    physicsBodyDefinitions [{
      boundBone {boneName=dyn_hair_rear_04}
      boxExtents=<20 30 30>
      constraintSetup {angularLimitsMax=<2 10 2>, angularLimitsMin=<-2 -10 -2>, angularTarget=<1 0 0>}
      localJointOffset=<4 0 0>
     }]
    simSpaceSettings {maxAngularAcceleration=30, maxAngularVelocity=770, simSpaceAngularAlpha=1}
   }
   n27 = Anim Dynamics - Bone: dyn_hair_rear_r_03(
    ComponentPose="Transform (Modify) Bone - Bone: dyn_hair_rear_03"(ComponentPose=@n26, Alpha=1, Translation="0,0,0", Rotation="0,0,0", Scale="1,0.6,1") {scaleMode=BMM_Additive, scaleSpace=BCS_BoneSpace, boneToModify {boneName=dyn_hair_rear_03}, scale {y=0.6}}
    Alpha=1
    ExternalForce="0,0,0"
   ) {
    angularDampingOverride=1
    angularSpringConstant=20000
    bAngularSpring=true
    bOverrideAngularDamping=true
    bUseGravityOverride=true
    gravityScale=0.35
    simulationSpace=RootRelative
    boundBone {boneName=dyn_hair_rear_r_03}
    componentAppliedLinearAccClamp=<0 0 0>
    gravityOverride {z=-980}
    physicsBodyDefinitions [{
      boundBone {boneName=dyn_hair_rear_r_03}
      boxExtents=<20 30 30>
      constraintSetup {angularLimitsMax=<2 10 2>, angularLimitsMin=<-2 -10 -2>, angularTarget=<1 0 0>}
      localJointOffset=<-4 0 0>
     }]
    simSpaceSettings {maxAngularAcceleration=30, maxAngularVelocity=770, simSpaceAngularAlpha=1}
   }
   n28 = Anim Dynamics - Bone: dyn_hair_rear_r_04(ComponentPose=@n27, Alpha=1, ExternalForce="0,0,0") {
    angularDampingOverride=1
    angularSpringConstant=20000
    bAngularSpring=true
    bOverrideAngularDamping=true
    bUseGravityOverride=true
    gravityScale=0.35
    simulationSpace=RootRelative
    boundBone {boneName=dyn_hair_rear_r_04}
    componentAppliedLinearAccClamp=<0 0 0>
    gravityOverride {z=-980}
    physicsBodyDefinitions [{
      boundBone {boneName=dyn_hair_rear_r_04}
      boxExtents=<20 30 30>
      constraintSetup {angularLimitsMax=<2 10 2>, angularLimitsMin=<-2 -10 -2>, angularTarget=<1 0 0>}
      localJointOffset=<-4 0 0>
     }]
    simSpaceSettings {maxAngularAcceleration=30, maxAngularVelocity=770, simSpaceAngularAlpha=1}
   }
   Save cached pose 'Backpack Sim Cache'(Pose=Component To Local(ComponentPose="Transform (Modify) Bone - Bone: dyn_hair_rear_r_02"(ComponentPose=@n28, Alpha=1, Translation="0,0,0", Rotation="0,0,0", Scale="1,0.7,1") {scaleMode=BMM_Additive, scaleSpace=BCS_BoneSpace, boneToModify {boneName=dyn_hair_rear_r_02}, scale {y=0.7}})) {cacheName=Backpack Sim Cache}
   n29 = RigidBody!(
    ComponentPose=Local To Component(LocalPose=Use cached pose 'Post Deformation'())
    Alpha
    OverrideWorldGravity="0,0,0"
    ExternalForce
    SimSpaceSettings:Sim Space Settings Structure=(WorldAlpha=0,VelocityScaleZ=1,DampingAlpha=1,MaxLinearVelocity=10000,MaxAngularVelocity=10000,MaxLinearAcceleration=10000,MaxAngularAcceleration=10000,ExternalLinearDragV=(X=0,Y=0,Z=0),ExternalLinearVelocity=(X=0,Y=0,Z=0),ExternalAngularVelocity=(X=0,Y=0,Z=0))
   ) -> (Pose) {
    bClampLinearTranslationLimitToRefPose=true
    bOverrideWorldGravity=true
    overridePhysicsAsset=&F_MED_VowelSleep_FaceAcc_Physics
    simulationSpace=BaseBoneSpace
    baseBoneRef {boneName=spine_04}
    componentAppliedLinearAccClamp=<0 0 1000>
    componentLinearAccScale {z=0.5}
    componentLinearVelScale {z=0.5}
   }
   n30 = "Transform (Modify) Bone - Bone: dyn_hair_rear_l_02"(ComponentPose=@n29, Alpha=1, Translation="0,-1,0", Rotation="0,0,0", Scale="1,1,1") {
    rotationMode=BMM_Additive
    rotationSpace=BCS_BoneSpace
    translationMode=BMM_Additive
    translationSpace=BCS_BoneSpace
    boneToModify {boneName=dyn_hair_rear_l_02}
    translation {y=-1}
   }
   n31 = "Transform (Modify) Bone - Bone: dyn_hair_rear_02"(ComponentPose=@n30, Alpha=1, Translation="0,-1,0", Rotation="0,0,0", Scale="1,1,1") {
    rotationMode=BMM_Additive
    rotationSpace=BCS_BoneSpace
    translationMode=BMM_Additive
    translationSpace=BCS_BoneSpace
    boneToModify {boneName=dyn_hair_rear_02}
    translation {y=-1}
   }
   n32 = "Transform (Modify) Bone - Bone: dyn_hair_rear_r_02"(ComponentPose=@n31, Alpha=1, Translation="0,1,0", Rotation="0,0,0", Scale="1,1,1") {
    rotationMode=BMM_Additive
    rotationSpace=BCS_BoneSpace
    translationMode=BMM_Additive
    translationSpace=BCS_BoneSpace
    boneToModify {boneName=dyn_hair_rear_r_02}
    translation {y=1}
   }
   n33 = "Transform (Modify) Bone - Bone: dyn_hair_rear_l_04"(ComponentPose=@n32, Alpha=1, Translation="0,0,0", Rotation="0,-5,0", Scale="1,1,1") {
    rotationMode=BMM_Additive
    rotationSpace=BCS_BoneSpace
    translationMode=BMM_Additive
    translationSpace=BCS_BoneSpace
    boneToModify {boneName=dyn_hair_rear_l_04}
    rotation {yaw=-5}
   }
   n34 = "Transform (Modify) Bone - Bone: dyn_hair_rear_04"(ComponentPose=@n33, Alpha=1, Translation="0,0,0", Rotation="0,-5,0", Scale="1,1,1") {
    rotationMode=BMM_Additive
    rotationSpace=BCS_BoneSpace
    translationMode=BMM_Additive
    translationSpace=BCS_BoneSpace
    boneToModify {boneName=dyn_hair_rear_04}
    rotation {yaw=-5}
   }
   n35 = "Transform (Modify) Bone - Bone: dyn_hair_rear_r_04"(ComponentPose=@n34, Alpha=1, Translation="0,0,0", Rotation="0,-5,0", Scale="1,1,1") {
    rotationMode=BMM_Additive
    rotationSpace=BCS_BoneSpace
    translationMode=BMM_Additive
    translationSpace=BCS_BoneSpace
    boneToModify {boneName=dyn_hair_rear_r_04}
    rotation {yaw=-5}
   }
   n36 = "Transform (Modify) Bone - Bone: dyn_hair_rear_l_02"!(ComponentPose=@n35, bAlphaBoolEnabled:bool=True, Translation="0,-5,0", Rotation="0,-2,0", Scale="1,1,1") -> (Pose) {
    alphaInputType=Bool
    rotationMode=BMM_Additive
    rotationSpace=BCS_BoneSpace
    translationMode=BMM_Additive
    translationSpace=BCS_BoneSpace
    boneToModify {boneName=dyn_hair_rear_l_02}
    rotation {yaw=-2}
    translation {y=-5}
   }
   n37 = "Transform (Modify) Bone - Bone: dyn_hair_rear_02"!(ComponentPose=@n36, bAlphaBoolEnabled:bool=True, Translation="0,-5,0", Rotation="0,-2,0", Scale="1,1,1") -> (Pose) {
    alphaInputType=Bool
    rotationMode=BMM_Additive
    rotationSpace=BCS_BoneSpace
    translationMode=BMM_Additive
    translationSpace=BCS_BoneSpace
    boneToModify {boneName=dyn_hair_rear_02}
    rotation {yaw=-2}
    translation {y=-5}
   }
   n38 = "Transform (Modify) Bone - Bone: dyn_hair_rear_r_02"!(ComponentPose=@n37, bAlphaBoolEnabled:bool=True, Translation="0,5,0", Rotation="0,-2,0", Scale="1,1,1") -> (Pose) {
    alphaInputType=Bool
    rotationMode=BMM_Additive
    rotationSpace=BCS_BoneSpace
    translationMode=BMM_Additive
    translationSpace=BCS_BoneSpace
    boneToModify {boneName=dyn_hair_rear_r_02}
    rotation {yaw=-2}
    translation {y=5}
   }
   n39 = "Transform (Modify) Bone - Bone: dyn_hair_rear_l_02"(ComponentPose=@n38, bAlphaBoolEnabled=True, Translation="0,-1,0", Rotation="0,-4,0", Scale="1,1,1") {
    alphaInputType=Bool
    rotationMode=BMM_Additive
    rotationSpace=BCS_BoneSpace
    translationMode=BMM_Additive
    translationSpace=BCS_BoneSpace
    boneToModify {boneName=dyn_hair_rear_l_02}
    rotation {yaw=-4}
    translation {y=-1}
   }
   n40 = "Transform (Modify) Bone - Bone: dyn_hair_rear_02"(ComponentPose=@n39, bAlphaBoolEnabled=True, Translation="0,-1,0", Rotation="0,-4,0", Scale="1,1,1") {
    alphaInputType=Bool
    rotationMode=BMM_Additive
    rotationSpace=BCS_BoneSpace
    translationMode=BMM_Additive
    translationSpace=BCS_BoneSpace
    boneToModify {boneName=dyn_hair_rear_02}
    rotation {yaw=-4}
    translation {y=-1}
   }
   n41 = "Transform (Modify) Bone - Bone: dyn_hair_rear_r_02"(ComponentPose=@n40, bAlphaBoolEnabled=True, Translation="0,1,0", Rotation="0,-4,0", Scale="1,1,1") {
    alphaInputType=Bool
    rotationMode=BMM_Additive
    rotationSpace=BCS_BoneSpace
    translationMode=BMM_Additive
    translationSpace=BCS_BoneSpace
    boneToModify {boneName=dyn_hair_rear_r_02}
    rotation {yaw=-4}
    translation {y=1}
   }
   Save cached pose 'Fitted Cape Sim Cache'(Pose=Component To Local(ComponentPose=@n41)) {cacheName=Fitted Cape Sim Cache}
  ]}
 EventGraph {nodes [
   Event BlueprintUpdateAnimation()
   > n1 = Sequence!(execute) -> (then_0, then_1, then_2:Exec)
   > Branch(Condition=Get bRigidBodyEnabled())
   > n2 = "Gravity Override State Function (Parameter Asset) (Nativized)"(
    CharacterPartAnimInstance=Self-Reference()
    GameTimeInSeconds=GetGameTimeInSeconds()
    Parameters=&/F_MED_VowelSleep_FaceAcc_Parameters
   )
   > Set Gravity Override(Gravity Override=@n2)
   n3 = NOT Boolean(A=Get bClothEnabled())
   n4 = NOT Boolean(A=Get bRigidBodyEnabled())
   Set Only Anim Dynamics(
    execute=@n1.then_1
    Only Anim Dynamics:bool=AND Boolean(A=Get bAnimDynamicsEnabled() -> (bAnimDynamicsEnabled:bool), B=AND Boolean(A=@n4, B=@n3))
   ) -> (Output_Get:bool)
   Set Fitted Cape(
    execute=@n1.then_2
    Fitted Cape:bool=BackpackItemDefinitionHasAnyMetaTag!(Tags:Array of Gameplay Tag Structures=Make Array("[0]"=(TagName="Cosmetics.MetaTags.VowelSleep")), bRequiresEquipped:bool=false) -> (ReturnValue:bool)
   ) -> (Output_Get:bool)
   Event OnCharacterCustomizationLoaded()
   n5 = CheckForBackbling()
   > Set isWearingBackbling(isWearingBackbling=OR Boolean(A=@n5, B=@n5.bIsBackblingLong))
   > Set Is Backbling A Cape(Is Backbling A Cape:bool=@n5.bIsBackblingACape) -> (Output_Get:bool)
   > Set Fitted Cape(Fitted Cape=BackpackItemDefinitionHasAnyMetaTag(Tags=Make Array("[0]"=(TagName="Cosmetics.MetaTags.VowelSleep")), bRequiresEquipped=false))
  ]}
]
variables [
 Fitted Cape {type=boolean}
 Gravity Override {type=Vector}
 Is Backbling A Cape {type=boolean}
 isWearingBackbling {type=boolean}
 Only Anim Dynamics {type=boolean}
 Sim Space Settings {
  type=SimSpaceSettings
  default {
   dampingAlpha=0.2
   maxAngularAcceleration=14
   maxAngularVelocity=2
   maxLinearAcceleration=2400
   maxLinearVelocity=770
   velocityScaleZ=0.5
   worldAlpha=0.9
   externalLinearDragV=<0.1 0.1 0.45>
  }
 }
]
